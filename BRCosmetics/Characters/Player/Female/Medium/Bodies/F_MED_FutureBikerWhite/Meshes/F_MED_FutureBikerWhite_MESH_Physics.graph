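# Physics asset for the FutureBikerWhite female medium body, setting up secondary motion for its collar, jacket and chains.
# Box bodies on dyn bones form short limited-swing, tight-twist chains from tiny MSTR anchor boxes, which connect to the pelvis with free-rotation constraints.
# Chain pieces also link back-to-mid-to-front on each side with free rotation.
# Capsules cover neck, pelvis and thighs, with extra front and back DSP capsules only on the right thigh.
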PhysicsAsset /BRCosmetics/Characters/Player/Female/Medium/Bodies/F_MED_FutureBikerWhite/Meshes/F_MED_FutureBikerWhite_MESH_Physics
bodies {
 dyn_chain_back_01_l [{shapeType=Box, extentX=8.5, extentY=12.5, extentZ=2.5, center=<9.1796 -0.0001 0.0004>}]
 dyn_chain_back_01_l_MSTR [{shapeType=Box, extentX=0.01, extentY=0.01, extentZ=0.01}]
 dyn_chain_back_01_r [{shapeType=Box, extentX=14, extentY=12.5, extentZ=1.5, center=<-7.4388 0.0001 -0.0004>, rotation=<rot 0 0 -180>}]
 dyn_chain_back_01_r_MSTR [{shapeType=Box, extentX=0.01, extentY=0.01, extentZ=0.01}]
 dyn_chain_front_01_l [{shapeType=Box, extentX=8.5, extentY=9, extentZ=3.5, center=<12.7058 0.0002 0.0004>}]
 dyn_chain_front_01_l_MSTR [{shapeType=Box, extentX=0.01, extentY=0.01, extentZ=0.01}]
 dyn_chain_front_01_r [{shapeType=Box, extentX=8.5, extentY=9, extentZ=3.5, center=<-12.7058 -0.0002 -0.0004>, rotation=<rot 0 0 180>}]
 dyn_chain_front_01_r_MSTR [{shapeType=Box, extentX=0.01, extentY=0.01, extentZ=0.01}]
 dyn_chain_mid_01_l [{shapeType=Box, extentX=8.5, extentY=19, extentZ=3.5, center=<15.2222 0 0.0004>}]
 dyn_chain_mid_01_l_MSTR [{shapeType=Box, extentX=0.01, extentY=0.01, extentZ=0.01}]
 dyn_chain_mid_01_r [{shapeType=Box, extentX=8.5, extentY=19, extentZ=3.5, center=<-15.2222 0 -0.0004>, rotation=<rot 0 0 180>}]
 dyn_chain_mid_01_r_MSTR [{shapeType=Box, extentX=0.01, extentY=0.01, extentZ=0.01}]
 dyn_collar_back_01 [{shapeType=Box, extentX=4.5, extentY=6, extentZ=2, center=<2.6751 0 0.0002>}]
 dyn_collar_back_01_MSTR [{shapeType=Box, extentX=0.01, extentY=0.01, extentZ=0.01}]
 dyn_collar_front_01_l [{shapeType=Box, extentX=4.5, extentY=6, extentZ=2, center=<2.6751 0 0.0002>}]
 dyn_collar_front_01_l_MSTR [{shapeType=Box, extentX=0.01, extentY=0.01, extentZ=0.01}]
 dyn_collar_front_01_r [{shapeType=Box, extentX=4.5, extentY=6, extentZ=2, center=<-2.6751 0 -0.0002>, rotation=<rot 0 0 180>}]
 dyn_collar_front_01_r_MSTR [{shapeType=Box, extentX=0.01, extentY=0.01, extentZ=0.01}]
 dyn_collar_front_02_l [{shapeType=Box, extentX=4.5, extentY=6, extentZ=2, center=<2.6751 0 0.0002>}]
 dyn_collar_front_02_r [{shapeType=Box, extentX=4.5, extentY=6, extentZ=2, center=<-2.6751 0 -0.0002>, rotation=<rot 0 0 179.9998>}]
 dyn_collar_side_01_l [{shapeType=Box, extentX=4.5, extentY=6, extentZ=2, center=<2.6751 0 0.0002>}]
 dyn_collar_side_01_l_MSTR [{shapeType=Box, extentX=0.01, extentY=0.01, extentZ=0.01}]
 dyn_collar_side_01_r [{shapeType=Box, extentX=4.5, extentY=6, extentZ=2, center=<-2.6751 0 -0.0002>, rotation=<rot 0 0 180>}]
 dyn_collar_side_01_r_MSTR [{shapeType=Box, extentX=0.01, extentY=0.01, extentZ=0.01}]
 dyn_jacket_back_01 [{shapeType=Box, extentX=4.5, extentY=6, extentZ=2, center=<2.6751 0 0.0002>}]
 dyn_jacket_back_01_MSTR [{shapeType=Box, extentX=0.01, extentY=0.01, extentZ=0.01}]
 dyn_jacket_back_02 [{shapeType=Box, extentX=4.5, extentY=6, extentZ=2, center=<2.6751 0 0.0002>}]
 dyn_jacket_front_01_l [{shapeType=Box, extentX=4.5, extentY=6, extentZ=2, center=<2.6751 0 0.0002>}]
 dyn_jacket_front_01_l_MSTR [{shapeType=Box, extentX=0.01, extentY=0.01, extentZ=0.01}]
 dyn_jacket_front_01_r [{shapeType=Box, extentX=4.5, extentY=6, extentZ=2, center=<-2.6751 0 -0.0002>, rotation=<rot 0 0 180>}]
 dyn_jacket_front_01_r_MSTR [{shapeType=Box, extentX=0.01, extentY=0.01, extentZ=0.01}]
 dyn_jacket_front_02_l [{shapeType=Box, extentX=4.5, extentY=6, extentZ=2, center=<2.6751 0 0.0002>}]
 dyn_jacket_front_02_r [{shapeType=Box, extentX=4.5, extentY=6, extentZ=2, center=<-2.6751 0 -0.0002>, rotation=<rot 0 0 180>}]
 dyn_jacket_front_03_l [{shapeType=Box, extentX=4.5, extentY=6, extentZ=2, center=<2.6751 0 0.0002>}]
 dyn_jacket_front_03_r [{shapeType=Box, extentX=4.5, extentY=6, extentZ=2, center=<-2.6751 0 -0.0002>, rotation=<rot 0 0 180>}]
 dyn_jacket_mid_01_l [{shapeType=Box, extentX=4.5, extentY=6, extentZ=2, center=<2.6751 0 0.0002>}]
 dyn_jacket_mid_01_l_MSTR [{shapeType=Box, extentX=0.01, extentY=0.01, extentZ=0.01}]
 dyn_jacket_mid_01_r [{shapeType=Box, extentX=4.5, extentY=6, extentZ=2, center=<-2.6751 0 -0.0002>, rotation=<rot 0 0 180>}]
 dyn_jacket_mid_01_r_MSTR [{shapeType=Box, extentX=0.01, extentY=0.01, extentZ=0.01}]
 dyn_jacket_side_01_l [{shapeType=Box, extentX=4.5, extentY=6, extentZ=2, center=<2.6751 0 0.0002>}]
 dyn_jacket_side_01_l_MSTR [{shapeType=Box, extentX=0.01, extentY=0.01, extentZ=0.01}]
 dyn_jacket_side_01_r [{shapeType=Box, extentX=4.5, extentY=6, extentZ=2, center=<-2.6751 0 -0.0002>, rotation=<rot 0 0 180>}]
 dyn_jacket_side_01_r_MSTR [{shapeType=Box, extentX=0.01, extentY=0.01, extentZ=0.01}]
 dyn_jacket_side_02_l [{shapeType=Box, extentX=4.5, extentY=6, extentZ=2, center=<2.6751 0 0.0002>}]
 dyn_jacket_side_02_r [{shapeType=Box, extentX=4.5, extentY=6, extentZ=2, center=<-2.6751 0 -0.0002>, rotation=<rot 0 0 180>}]
 dyn_jacket_upper_01_l [{shapeType=Box, extentX=4.5, extentY=6, extentZ=2, center=<2.6751 0 0.0002>}]
 dyn_jacket_upper_01_l_MSTR [{shapeType=Box, extentX=0.01, extentY=0.01, extentZ=0.01}]
 dyn_jacket_upper_01_r [{shapeType=Box, extentX=4.5, extentY=6, extentZ=2, center=<-2.6751 0 -0.0002>, rotation=<rot 0 0 180>}]
 dyn_jacket_upper_01_r_MSTR [{shapeType=Box, extentX=0.01, extentY=0.01, extentZ=0.01}]
 dyn_jacket_upper_02_l [{shapeType=Box, extentX=4.5, extentY=6, extentZ=2, center=<2.6751 0 0.0002>}]
 dyn_jacket_upper_02_r [{shapeType=Box, extentX=4.5, extentY=6, extentZ=2, center=<-2.6751 0 -0.0002>, rotation=<rot 0 0 180>}]
 neck_01 [{shapeType=Capsule, length=9.3476, radius=7.34, center=<8.8043 1.8311 0.0005>, rotation=<rot -90 0 -20.0008>}]
 neck_02 [{shapeType=Capsule, length=10.3476, radius=6.34, center=<4.5845 5.2555 0>, rotation=<rot -90 -0.0049 0.0049>}]
 pelvis [{shapeType=Capsule, length=11.6371, radius=10.1854, center=<-6.9465 -2.7596 0.0071>, rotation=<rot -0.2337 -89.9999 0.1497>}]
 thigh_back_r_01_DSP [{shapeType=Capsule, length=14.3476, radius=7.84, center=<9.1431 -4.657 -0.7848>, rotation=<rot -90 7.125 -52.1172>}]
 thigh_front_r_01_DSP [{shapeType=Capsule, length=14.3476, radius=7.84, center=<9.7087 1.1949 -0.7134>, rotation=<rot -90 -0.0148 0.0148>}]
 thigh_l [{shapeType=Capsule, length=15.6876, radius=7.75, center=<-8.1311 -1.3854 -0.0003>, rotation=<rot -90 179.9802 -179.9802>}]
 thigh_r [{shapeType=Capsule, length=15.6876, radius=7.75, center=<8.1311 1.3854 0.0003>, rotation=<rot -90 -0.0066 -179.9934>}]
 thigh_twist_01_l [{shapeType=Capsule, length=7.8476, radius=8.84, center=<11.5987 -0.8847 0.9815>, rotation=<rot -90 -0.0264 0.0264>}]
 thigh_twist_01_r [{shapeType=Capsule, length=7.8476, radius=8.84, center=<-11.5987 0.8847 -0.9815>, rotation=<rot -90 179.9802 0.0198>}]
}
constraints [
 {bone1Name=dyn_collar_front_01_l, bone2Name=dyn_collar_front_01_l_MSTR, swing1LimitDegrees=5, swing2LimitDegrees=45, twistLimitDegrees=5}
 {bone1Name=dyn_collar_front_02_l, bone2Name=dyn_collar_front_01_l, swing1LimitDegrees=5, swing2LimitDegrees=45, twistLimitDegrees=5}
 {bone1Name=dyn_collar_side_01_l, bone2Name=dyn_collar_side_01_l_MSTR, swing1LimitDegrees=5, swing2LimitDegrees=25, twistLimitDegrees=5}
 {bone1Name=dyn_collar_back_01, bone2Name=dyn_collar_back_01_MSTR, swing1LimitDegrees=5, swing2LimitDegrees=25, twistLimitDegrees=5}
 {bone1Name=dyn_collar_side_01_r, bone2Name=dyn_collar_side_01_r_MSTR, swing1LimitDegrees=5, swing2LimitDegrees=25, twistLimitDegrees=5}
 {bone1Name=dyn_collar_front_01_r, bone2Name=dyn_collar_front_01_r_MSTR, swing1LimitDegrees=5, swing2LimitDegrees=45, twistLimitDegrees=5}
 {bone1Name=dyn_collar_front_02_r, bone2Name=dyn_collar_front_01_r, swing1LimitDegrees=5, swing2LimitDegrees=45, twistLimitDegrees=5}
 {bone1Name=dyn_jacket_upper_01_l, bone2Name=dyn_jacket_upper_01_l_MSTR, swing1LimitDegrees=5, swing2LimitDegrees=35, twistLimitDegrees=5}
 {bone1Name=dyn_jacket_upper_02_l, bone2Name=dyn_jacket_upper_01_l, swing1LimitDegrees=5, swing2LimitDegrees=35, twistLimitDegrees=5}
 {bone1Name=dyn_jacket_upper_02_r, bone2Name=dyn_jacket_upper_01_r, swing1LimitDegrees=5, swing2LimitDegrees=35, twistLimitDegrees=5}
 {bone1Name=dyn_jacket_upper_01_r, bone2Name=dyn_jacket_upper_01_r_MSTR, swing1LimitDegrees=5, swing2LimitDegrees=35, twistLimitDegrees=5}
 {
  bone1Name=dyn_jacket_mid_01_r
  bone2Name=dyn_jacket_mid_01_r_MSTR
  swing1LimitDegrees=5
  swing1Motion=Locked
  swing2LimitDegrees=35
  twistLimitDegrees=5
 }
 {
  bone1Name=dyn_jacket_mid_01_l
  bone2Name=dyn_jacket_mid_01_l_MSTR
  swing1LimitDegrees=5
  swing1Motion=Locked
  swing2LimitDegrees=35
  twistLimitDegrees=5
 }
 {bone1Name=dyn_jacket_back_01, bone2Name=dyn_jacket_back_01_MSTR, swing1LimitDegrees=5, swing2LimitDegrees=15, twistLimitDegrees=5}
 {bone1Name=dyn_jacket_back_02, bone2Name=dyn_jacket_back_01, swing1LimitDegrees=5, swing2LimitDegrees=15, twistLimitDegrees=5}
 {bone1Name=dyn_jacket_front_01_l, bone2Name=dyn_jacket_front_01_l_MSTR, swing1LimitDegrees=5, swing2LimitDegrees=35, twistLimitDegrees=5}
 {bone1Name=dyn_jacket_front_02_l, bone2Name=dyn_jacket_front_01_l, swing1LimitDegrees=5, swing2LimitDegrees=35, twistLimitDegrees=5}
 {bone1Name=dyn_jacket_front_03_l, bone2Name=dyn_jacket_front_02_l, swing1LimitDegrees=35, swing2LimitDegrees=35, twistLimitDegrees=5}
 {bone1Name=dyn_jacket_side_01_l, bone2Name=dyn_jacket_side_01_l_MSTR, swing1LimitDegrees=5, swing2LimitDegrees=15, twistLimitDegrees=5}
 {bone1Name=dyn_jacket_side_02_l, bone2Name=dyn_jacket_side_01_l, swing1LimitDegrees=5, swing2LimitDegrees=15, twistLimitDegrees=5}
 {bone1Name=dyn_jacket_side_01_r, bone2Name=dyn_jacket_side_01_r_MSTR, swing1LimitDegrees=5, swing2LimitDegrees=15, twistLimitDegrees=5}
 {bone1Name=dyn_jacket_side_02_r, bone2Name=dyn_jacket_side_01_r, swing1LimitDegrees=5, swing2LimitDegrees=15, twistLimitDegrees=5}
 {bone1Name=dyn_jacket_front_01_r, bone2Name=dyn_jacket_front_01_r_MSTR, swing1LimitDegrees=5, swing2LimitDegrees=35, twistLimitDegrees=5}
 {bone1Name=dyn_jacket_front_02_r, bone2Name=dyn_jacket_front_01_r, swing1LimitDegrees=5, swing2LimitDegrees=35, twistLimitDegrees=5}
 {bone1Name=dyn_jacket_front_03_r, bone2Name=dyn_jacket_front_02_r, swing1LimitDegrees=35, swing2LimitDegrees=35, twistLimitDegrees=5}
 {bone1Name=dyn_chain_back_01_l, bone2Name=dyn_chain_back_01_l_MSTR, swing1LimitDegrees=15, swing2LimitDegrees=45, twistLimitDegrees=5}
 {bone1Name=dyn_chain_mid_01_l, bone2Name=dyn_chain_mid_01_l_MSTR, swing1LimitDegrees=75, swing2LimitDegrees=75, twistLimitDegrees=5}
 {bone1Name=dyn_chain_front_01_l, bone2Name=dyn_chain_front_01_l_MSTR, swing1LimitDegrees=75, swing2LimitDegrees=75, twistLimitDegrees=5}
 {
  bone1Name=dyn_collar_front_01_r_MSTR
  bone2Name=pelvis
  swing1LimitDegrees=45
  swing1Motion=Free
  swing2LimitDegrees=45
  swing2Motion=Free
  twistLimitDegrees=45
  twistMotion=Free
 }
 {
  bone1Name=dyn_collar_front_01_l_MSTR
  bone2Name=pelvis
  swing1LimitDegrees=45
  swing1Motion=Free
  swing2LimitDegrees=45
  swing2Motion=Free
  twistLimitDegrees=45
  twistMotion=Free
 }
 {
  bone1Name=dyn_collar_side_01_r_MSTR
  bone2Name=pelvis
  swing1LimitDegrees=45
  swing1Motion=Free
  swing2LimitDegrees=45
  swing2Motion=Free
  twistLimitDegrees=45
  twistMotion=Free
 }
 {
  bone1Name=dyn_collar_side_01_l_MSTR
  bone2Name=pelvis
  swing1LimitDegrees=45
  swing1Motion=Free
  swing2LimitDegrees=45
  swing2Motion=Free
  twistLimitDegrees=45
  twistMotion=Free
 }
 {
  bone1Name=dyn_collar_back_01_MSTR
  bone2Name=pelvis
  swing1LimitDegrees=45
  swing1Motion=Free
  swing2LimitDegrees=45
  swing2Motion=Free
  twistLimitDegrees=45
  twistMotion=Free
 }
 {
  bone1Name=dyn_jacket_upper_01_l_MSTR
  bone2Name=pelvis
  swing1LimitDegrees=45
  swing1Motion=Free
  swing2LimitDegrees=45
  swing2Motion=Free
  twistLimitDegrees=45
  twistMotion=Free
 }
 {
  bone1Name=dyn_jacket_upper_01_r_MSTR
  bone2Name=pelvis
  swing1LimitDegrees=45
  swing1Motion=Free
  swing2LimitDegrees=45
  swing2Motion=Free
  twistLimitDegrees=45
  twistMotion=Free
 }
 {
  bone1Name=dyn_jacket_mid_01_r_MSTR
  bone2Name=pelvis
  swing1LimitDegrees=45
  swing1Motion=Free
  swing2LimitDegrees=45
  swing2Motion=Free
  twistLimitDegrees=45
  twistMotion=Free
 }
 {
  bone1Name=dyn_jacket_mid_01_l_MSTR
  bone2Name=pelvis
  swing1LimitDegrees=45
  swing1Motion=Free
  swing2LimitDegrees=45
  swing2Motion=Free
  twistLimitDegrees=45
  twistMotion=Free
 }
 {
  bone1Name=dyn_jacket_side_01_l_MSTR
  bone2Name=pelvis
  swing1LimitDegrees=45
  swing1Motion=Free
  swing2LimitDegrees=45
  swing2Motion=Free
  twistLimitDegrees=45
  twistMotion=Free
 }
 {
  bone1Name=dyn_jacket_front_01_l_MSTR
  bone2Name=pelvis
  swing1LimitDegrees=45
  swing1Motion=Free
  swing2LimitDegrees=45
  swing2Motion=Free
  twistLimitDegrees=45
  twistMotion=Free
 }
 {
  bone1Name=dyn_jacket_back_01_MSTR
  bone2Name=pelvis
  swing1LimitDegrees=45
  swing1Motion=Free
  swing2LimitDegrees=45
  swing2Motion=Free
  twistLimitDegrees=45
  twistMotion=Free
 }
 {
  bone1Name=dyn_jacket_side_01_r_MSTR
  bone2Name=pelvis
  swing1LimitDegrees=45
  swing1Motion=Free
  swing2LimitDegrees=45
  swing2Motion=Free
  twistLimitDegrees=45
  twistMotion=Free
 }
 {
  bone1Name=dyn_jacket_front_01_r_MSTR
  bone2Name=pelvis
  swing1LimitDegrees=45
  swing1Motion=Free
  swing2LimitDegrees=45
  swing2Motion=Free
  twistLimitDegrees=45
  twistMotion=Free
 }
 {
  bone1Name=thigh_l
  bone2Name=pelvis
  swing1LimitDegrees=45
  swing1Motion=Free
  swing2LimitDegrees=45
  swing2Motion=Free
  twistLimitDegrees=45
  twistMotion=Free
 }
 {
  bone1Name=dyn_chain_back_01_l_MSTR
  bone2Name=pelvis
  swing1LimitDegrees=45
  swing1Motion=Free
  swing2LimitDegrees=45
  swing2Motion=Free
  twistLimitDegrees=45
  twistMotion=Free
 }
 {
  bone1Name=dyn_chain_mid_01_l_MSTR
  bone2Name=pelvis
  swing1LimitDegrees=45
  swing1Motion=Free
  swing2LimitDegrees=45
  swing2Motion=Free
  twistLimitDegrees=45
  twistMotion=Free
 }
 {
  bone1Name=dyn_chain_front_01_l_MSTR
  bone2Name=pelvis
  swing1LimitDegrees=45
  swing1Motion=Free
  swing2LimitDegrees=45
  swing2Motion=Free
  twistLimitDegrees=45
  twistMotion=Free
 }
 {
  bone1Name=dyn_chain_back_01_r_MSTR
  bone2Name=pelvis
  swing1LimitDegrees=45
  swing1Motion=Free
  swing2LimitDegrees=45
  swing2Motion=Free
  twistLimitDegrees=45
  twistMotion=Free
 }
 {
  bone1Name=dyn_chain_mid_01_r_MSTR
  bone2Name=pelvis
  swing1LimitDegrees=45
  swing1Motion=Free
  swing2LimitDegrees=45
  swing2Motion=Free
  twistLimitDegrees=45
  twistMotion=Free
 }
 {
  bone1Name=dyn_chain_front_01_r_MSTR
  bone2Name=pelvis
  swing1LimitDegrees=45
  swing1Motion=Free
  swing2LimitDegrees=45
  swing2Motion=Free
  twistLimitDegrees=45
  twistMotion=Free
 }
 {
  bone1Name=dyn_chain_back_01_l
  bone2Name=dyn_chain_mid_01_l
  swing1LimitDegrees=45
  swing1Motion=Free
  swing2LimitDegrees=45
  swing2Motion=Free
  twistLimitDegrees=45
  twistMotion=Free
 }
 {
  bone1Name=dyn_chain_mid_01_l
  bone2Name=dyn_chain_front_01_l
  swing1LimitDegrees=45
  swing1Motion=Free
  swing2LimitDegrees=45
  swing2Motion=Free
  twistLimitDegrees=45
  twistMotion=Free
 }
 {
  bone1Name=thigh_twist_01_l
  bone2Name=thigh_l
  swing1LimitDegrees=45
  swing1Motion=Free
  swing2LimitDegrees=45
  swing2Motion=Free
  twistLimitDegrees=45
  twistMotion=Free
 }
 {bone1Name=dyn_chain_front_01_r, bone2Name=dyn_chain_front_01_r_MSTR, swing1LimitDegrees=75, swing2LimitDegrees=75, twistLimitDegrees=5}
 {
  bone1Name=thigh_twist_01_r
  bone2Name=thigh_r
  swing1LimitDegrees=45
  swing1Motion=Free
  swing2LimitDegrees=45
  swing2Motion=Free
  twistLimitDegrees=45
  twistMotion=Free
 }
 {
  bone1Name=thigh_r
  bone2Name=pelvis
  swing1LimitDegrees=45
  swing1Motion=Free
  swing2LimitDegrees=45
  swing2Motion=Free
  twistLimitDegrees=45
  twistMotion=Free
 }
 {
  bone1Name=neck_02
  bone2Name=neck_01
  swing1LimitDegrees=45
  swing1Motion=Free
  swing2LimitDegrees=45
  swing2Motion=Free
  twistLimitDegrees=45
  twistMotion=Free
 }
 {
  bone1Name=neck_01
  bone2Name=pelvis
  swing1LimitDegrees=45
  swing1Motion=Free
  swing2LimitDegrees=45
  swing2Motion=Free
  twistLimitDegrees=45
  twistMotion=Free
 }
 {
  bone1Name=thigh_back_r_01_DSP
  bone2Name=thigh_r
  swing1LimitDegrees=45
  swing1Motion=Free
  swing2LimitDegrees=45
  swing2Motion=Free
  twistLimitDegrees=45
  twistMotion=Free
 }
 {
  bone1Name=thigh_front_r_01_DSP
  bone2Name=thigh_r
  swing1LimitDegrees=45
  swing1Motion=Free
  swing2LimitDegrees=45
  swing2Motion=Free
  twistLimitDegrees=45
  twistMotion=Free
 }
 {bone1Name=dyn_chain_back_01_r, bone2Name=dyn_chain_back_01_r_MSTR, swing1LimitDegrees=45, swing2LimitDegrees=45, twistLimitDegrees=5}
 {bone1Name=dyn_chain_mid_01_r, bone2Name=dyn_chain_mid_01_r_MSTR, swing1LimitDegrees=45, swing2LimitDegrees=75, twistLimitDegrees=5}
 {
  bone1Name=dyn_chain_back_01_r
  bone2Name=dyn_chain_mid_01_r
  swing1LimitDegrees=45
  swing1Motion=Free
  swing2LimitDegrees=45
  swing2Motion=Free
  twistLimitDegrees=45
  twistMotion=Free
 }
 {
  bone1Name=dyn_chain_front_01_r
  bone2Name=dyn_chain_mid_01_r
  swing1LimitDegrees=45
  swing1Motion=Free
  swing2LimitDegrees=45
  swing2Motion=Free
  twistLimitDegrees=45
  twistMotion=Free
 }
]
