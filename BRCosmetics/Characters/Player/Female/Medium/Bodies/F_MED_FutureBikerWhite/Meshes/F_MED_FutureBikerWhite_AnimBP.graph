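# Cosmetic AnimBlueprint for the F_MED_FutureBikerWhite body, adding secondary motion and pose correctives to chains, jacket, collar and pads.
# It copies the parent mesh pose, corrects thigh and collar bones, then caches it for RigidBody simulation with skydiving wind.
# Post-simulation bone offsets and weighted constraints spread dynamic bones across chain links and adjust pads.
# The EventGraph derives blend alphas from joint deltas and movement state, resetting dynamics on crouch changes.
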AnimBlueprint /BRCosmetics/Characters/Player/Female/Medium/Bodies/F_MED_FutureBikerWhite/Meshes/F_MED_FutureBikerWhite_AnimBP
target_skeleton=&/F_MED_FutureBikerWhite_Skeleton
graphs [
 AnimGraph {nodes [
   n1 = "Constraint - chain_01_r_DSP (PP)"!(ComponentPose, Alpha=1, ConstraintWeights_0=0.7, ConstraintWeights_1:float=0.3) -> (Pose) {
    alphaBoolBlend {blendOption=Linear}
    boneToModify {boneName=chain_01_r_DSP}
    constraintSetup [{targetBone {boneName=dyn_chain_front_01_r}}, {targetBone {boneName=dyn_chain_mid_01_r}}]
    constraintWeights [1,1]
   }
   n2 = "Constraint - chain_02_r_DSP (PP)"!(ComponentPose=@n1, Alpha=1, ConstraintWeights_0=0.3, ConstraintWeights_1:float=0.7) -> (Pose) {
    alphaBoolBlend {blendOption=Linear}
    boneToModify {boneName=chain_02_r_DSP}
    constraintSetup [{targetBone {boneName=dyn_chain_front_01_r}}, {targetBone {boneName=dyn_chain_mid_01_r}}]
    constraintWeights [1,1]
   }
   n3 = "Constraint - chain_03_r_DSP (PP)"!(ComponentPose=@n2, Alpha=1, ConstraintWeights_0=0.2, ConstraintWeights_1:float=0.8) -> (Pose) {
    alphaBoolBlend {blendOption=Linear}
    boneToModify {boneName=chain_03_r_DSP}
    constraintSetup [{targetBone {boneName=dyn_chain_front_01_r}}, {targetBone {boneName=dyn_chain_mid_01_r}}]
    constraintWeights [1,1]
   }
   n4 = "Constraint - chain_04_r_DSP (P)"(ComponentPose=@n3, Alpha=1, ConstraintWeights_0=1) {
    alphaBoolBlend {blendOption=Linear}
    boneToModify {boneName=chain_04_r_DSP}
    constraintSetup [{targetBone {boneName=dyn_chain_mid_01_r}}]
    constraintWeights [1]
   }
   n5 = "Constraint - chain_05_r_DSP (PP)"!(ComponentPose=@n4, Alpha=1, ConstraintWeights_0=0.2, ConstraintWeights_1:float=0.8) -> (Pose) {
    alphaBoolBlend {blendOption=Linear}
    boneToModify {boneName=chain_05_r_DSP}
    constraintSetup [{targetBone {boneName=dyn_chain_back_01_r}}, {targetBone {boneName=dyn_chain_mid_01_r}}]
    constraintWeights [1,1]
   }
   n6 = "Constraint - chain_06_r_DSP (PP)"!(ComponentPose=@n5, Alpha=1, ConstraintWeights_0=0.3, ConstraintWeights_1:float=0.7) -> (Pose) {
    alphaBoolBlend {blendOption=Linear}
    boneToModify {boneName=chain_06_r_DSP}
    constraintSetup [{targetBone {boneName=dyn_chain_back_01_r}}, {targetBone {boneName=dyn_chain_mid_01_r}}]
    constraintWeights [1,1]
   }
   n7 = "Constraint - chain_07_r_DSP (PP)"!(ComponentPose=@n6, Alpha=1, ConstraintWeights_0=0.3, ConstraintWeights_1:float=0.7) -> (Pose) {
    alphaBoolBlend {blendOption=Linear}
    boneToModify {boneName=chain_07_r_DSP}
    constraintSetup [{targetBone {boneName=dyn_chain_back_01_r}}, {targetBone {boneName=dyn_chain_mid_01_r}}]
    constraintWeights [1,1]
   }
   n8 = "Constraint - chain_08_r_DSP (PP)"!(ComponentPose=@n7, Alpha=1, ConstraintWeights_0=0.7, ConstraintWeights_1:float=0.3) -> (Pose) {
    alphaBoolBlend {blendOption=Linear}
    boneToModify {boneName=chain_08_r_DSP}
    constraintSetup [{targetBone {boneName=dyn_chain_back_01_r}}, {targetBone {boneName=dyn_chain_mid_01_r}}]
    constraintWeights [1,1]
   }
   "Constraint - chain_09_r_DSP (P)"(ComponentPose=@n8, Alpha=1, ConstraintWeights_0=1) {
    alphaBoolBlend {blendOption=Linear}
    boneToModify {boneName=chain_09_r_DSP}
    constraintSetup [{targetBone {boneName=dyn_chain_back_01_r}}]
    constraintWeights [1]
   }
   "Transform (Modify) Bone - Bone: thigh_back_l_01_DSP"(Alpha=1, Translation="0,0,0", Rotation="0.000,0.000,60", Scale="1.25,1.000,1.000") {
    rotationMode=BMM_Additive
    scaleMode=BMM_Additive
    translationMode=BMM_Additive
    translationSpace=BCS_BoneSpace
    alphaBoolBlend {blendOption=Linear}
    boneToModify {boneName=thigh_back_l_01_DSP}
    rotation {roll=60}
    scale {x=1.25}
   }
   "Transform (Modify) Bone - Bone: dyn_chain_mid_01_r"(
    Alpha=Get thighSwingDelta_r_post_translation() -> (thighSwingDelta_r_post_translation:double)
    Translation="15,0.000,0.000"
    Rotation="0,0,0"
    Scale="1,1,1"
   ) {
    rotationMode=BMM_Additive
    rotationSpace=BCS_BoneSpace
    scaleMode=BMM_Additive
    scaleSpace=BCS_BoneSpace
    translationMode=BMM_Additive
    translationSpace=BCS_BoneSpace
    alphaBoolBlend {blendOption=Linear}
    boneToModify {boneName=dyn_chain_mid_01_r}
    translation {x=15}
   }
   n9 = RigidBody(ComponentPose=Local To Component(LocalPose=Use cached pose 'preSimCache'()), ExternalForce=Get windVector() -> (windVector:Vector)) {
    bClampLinearTranslationLimitToRefPose=true
    bOverrideWorldGravity=true
    overridePhysicsAsset=&/F_MED_FutureBikerWhite_MESH_Physics
    simulationSpace=BaseBoneSpace
    alphaBoolBlend {blendOption=Linear}
    baseBoneRef {boneName=spine_03}
    componentAppliedLinearAccClamp=<3000 3000 3000>
    componentLinearAccScale=<0.5 0.5 0.5>
    componentLinearVelScale=<0.7 0.7 0.7>
    overrideWorldGravity {z=-980}
   }
   n10 = "Transform (Modify) Bone - Bone: dyn_chain_back_01_r"(
    ComponentPose=@n9
    Alpha=float * float(A=Get thighSwingDelta_r_post() -> (thighSwingDelta_r_post:double), B=0.5)
    Translation="0,0,0"
    Rotation="30,22,0"
    Scale="1,1,1"
   ) {
    rotationMode=BMM_Additive
    rotationSpace=BCS_BoneSpace
    scaleMode=BMM_Replace
    scaleSpace=BCS_BoneSpace
    translationMode=BMM_Additive
    translationSpace=BCS_BoneSpace
    alphaBoolBlend {blendOption=Linear}
    boneToModify {boneName=dyn_chain_back_01_r}
    rotation {pitch=30, yaw=22}
   }
   n11 = "Transform (Modify) Bone - Bone: dyn_chain_mid_01_r"(ComponentPose=@n10, Alpha=0.5, Translation="0,0,0", Rotation="10,0.000,0", Scale="1,1,1") {
    rotationMode=BMM_Additive
    rotationSpace=BCS_BoneSpace
    scaleMode=BMM_Additive
    scaleSpace=BCS_BoneSpace
    translationMode=BMM_Additive
    translationSpace=BCS_BoneSpace
    alphaBoolBlend {blendOption=Linear}
    boneToModify {boneName=dyn_chain_mid_01_r}
    rotation {pitch=10}
   }
   n12 = "Transform (Modify) Bone - Bone: dyn_chain_front_01_r"(
    ComponentPose=@n11
    Alpha=Get thighSwingDelta_r_post_mid_rotation() -> (thighSwingDelta_r_post_mid_rotation:double)
    Translation="0,0,0"
    Rotation="30,0.000,0"
    Scale="1,1,1"
   ) {
    rotationMode=BMM_Additive
    rotationSpace=BCS_BoneSpace
    scaleMode=BMM_Additive
    scaleSpace=BCS_BoneSpace
    translationMode=BMM_Additive
    translationSpace=BCS_BoneSpace
    alphaBoolBlend {blendOption=Linear}
    boneToModify {boneName=dyn_chain_front_01_r}
    rotation {pitch=30}
   }
   n13 = "Transform (Modify) Bone - Bone: dyn_chain_back_01_l"(ComponentPose=@n12, Alpha=0.5, Translation="0,0,0", Rotation="30,22,0", Scale="1,1,1") {
    rotationMode=BMM_Additive
    rotationSpace=BCS_BoneSpace
    scaleMode=BMM_Additive
    translationMode=BMM_Additive
    translationSpace=BCS_BoneSpace
    alphaBoolBlend {blendOption=Linear}
    boneToModify {boneName=dyn_chain_back_01_l}
    rotation {pitch=30, yaw=22}
   }
   n14 = "Constraint - chain_01_l_DSP (PP)"!(ComponentPose=@n13, Alpha=1, ConstraintWeights_0=0.7, ConstraintWeights_1:float=0.3) -> (Pose) {
    alphaBoolBlend {blendOption=Linear}
    boneToModify {boneName=chain_01_l_DSP}
    constraintSetup [{targetBone {boneName=dyn_chain_front_01_l}}, {targetBone {boneName=dyn_chain_mid_01_l}}]
    constraintWeights [1,1]
   }
   n15 = "Constraint - chain_02_l_DSP (PP)"!(ComponentPose=@n14, Alpha=1, ConstraintWeights_0=0.3, ConstraintWeights_1:float=0.7) -> (Pose) {
    alphaBoolBlend {blendOption=Linear}
    boneToModify {boneName=chain_02_l_DSP}
    constraintSetup [{targetBone {boneName=dyn_chain_front_01_l}}, {targetBone {boneName=dyn_chain_mid_01_l}}]
    constraintWeights [1,1]
   }
   n16 = "Constraint - chain_03_l_DSP (PP)"!(ComponentPose=@n15, Alpha=1, ConstraintWeights_0=0.2, ConstraintWeights_1:float=0.8) -> (Pose) {
    alphaBoolBlend {blendOption=Linear}
    boneToModify {boneName=chain_03_l_DSP}
    constraintSetup [{targetBone {boneName=dyn_chain_front_01_l}}, {targetBone {boneName=dyn_chain_mid_01_l}}]
    constraintWeights [1,1]
   }
   n17 = "Constraint - chain_04_l_DSP (P)"(ComponentPose=@n16, Alpha=1, ConstraintWeights_0=1) {
    alphaBoolBlend {blendOption=Linear}
    boneToModify {boneName=chain_04_l_DSP}
    constraintSetup [{targetBone {boneName=dyn_chain_mid_01_l}}]
    constraintWeights [1]
   }
   n18 = "Constraint - chain_05_l_DSP (PP)"!(ComponentPose=@n17, Alpha=1, ConstraintWeights_0=0.2, ConstraintWeights_1:float=0.8) -> (Pose) {
    alphaBoolBlend {blendOption=Linear}
    boneToModify {boneName=chain_05_l_DSP}
    constraintSetup [{targetBone {boneName=dyn_chain_back_01_l}}, {targetBone {boneName=dyn_chain_mid_01_l}}]
    constraintWeights [1,1]
   }
   n19 = "Constraint - chain_06_l_DSP (PP)"!(ComponentPose=@n18, Alpha=1, ConstraintWeights_0=0.3, ConstraintWeights_1:float=0.7) -> (Pose) {
    alphaBoolBlend {blendOption=Linear}
    boneToModify {boneName=chain_06_l_DSP}
    constraintSetup [{targetBone {boneName=dyn_chain_back_01_l}}, {targetBone {boneName=dyn_chain_mid_01_l}}]
    constraintWeights [1,1]
   }
   n20 = "Constraint - chain_07_l_DSP (PP)"!(ComponentPose=@n19, Alpha=1, ConstraintWeights_0=0.3, ConstraintWeights_1:float=0.7) -> (Pose) {
    alphaBoolBlend {blendOption=Linear}
    boneToModify {boneName=chain_07_l_DSP}
    constraintSetup [{targetBone {boneName=dyn_chain_back_01_l}}, {targetBone {boneName=dyn_chain_mid_01_l}}]
    constraintWeights [1,1]
   }
   n21 = "Constraint - chain_08_l_DSP (PP)"!(ComponentPose=@n20, Alpha=1, ConstraintWeights_0=0.7, ConstraintWeights_1:float=0.3) -> (Pose) {
    alphaBoolBlend {blendOption=Linear}
    boneToModify {boneName=chain_08_l_DSP}
    constraintSetup [{targetBone {boneName=dyn_chain_back_01_l}}, {targetBone {boneName=dyn_chain_mid_01_l}}]
    constraintWeights [1,1]
   }
   n22 = "Constraint - chain_09_l_DSP (P)"(ComponentPose=@n21, Alpha=1, ConstraintWeights_0=1) {
    alphaBoolBlend {blendOption=Linear}
    boneToModify {boneName=chain_09_l_DSP}
    constraintSetup [{targetBone {boneName=dyn_chain_back_01_l}}]
    constraintWeights [1]
   }
   n23 = "Constraint - chain_09_r_DSP (P)"(ComponentPose=@n22, Alpha=1, ConstraintWeights_0=1) {
    alphaBoolBlend {blendOption=Linear}
    boneToModify {boneName=chain_09_r_DSP}
    constraintSetup [{targetBone {boneName=dyn_chain_back_01_r}}]
    constraintWeights [1]
   }
   n24 = "Constraint - chain_08_r_DSP (PPP)"!(ComponentPose=@n23, Alpha=1, ConstraintWeights_0=0.2, ConstraintWeights_1:float=0, ConstraintWeights_2:float=0.8) -> (Pose) {
    alphaBoolBlend {blendOption=Linear}
    boneToModify {boneName=chain_08_r_DSP}
    constraintSetup [{targetBone {boneName=dyn_chain_back_01_r}}, {targetBone {boneName=dyn_chain_mid_01_r}}, {targetBone {boneName=chain_09_r_DSP}}]
    constraintWeights [1,1,1]
   }
   n25 = "Constraint - chain_07_r_DSP (PPP)"!(ComponentPose=@n24, Alpha=1, ConstraintWeights_0=0.2, ConstraintWeights_1:float=0.2, ConstraintWeights_2:float=0.6) -> (Pose) {
    alphaBoolBlend {blendOption=Linear}
    boneToModify {boneName=chain_07_r_DSP}
    constraintSetup [{targetBone {boneName=dyn_chain_back_01_r}}, {targetBone {boneName=dyn_chain_mid_01_r}}, {targetBone {boneName=chain_08_r_DSP}}]
    constraintWeights [1,1,1]
   }
   n26 = "Constraint - chain_04_r_DSP (P)"(ComponentPose=@n25, Alpha=1, ConstraintWeights_0=1) {
    alphaBoolBlend {blendOption=Linear}
    boneToModify {boneName=chain_04_r_DSP}
    constraintSetup [{targetBone {boneName=dyn_chain_mid_01_r}}]
    constraintWeights [1]
   }
   n27 = "Constraint - chain_03_r_DSP (PP)"(ComponentPose=@n26, Alpha=1, ConstraintWeights_0=0.2, ConstraintWeights_1=0.8) {
    alphaBoolBlend {blendOption=Linear}
    boneToModify {boneName=chain_03_r_DSP}
    constraintSetup [{targetBone {boneName=dyn_chain_front_01_r}}, {targetBone {boneName=dyn_chain_mid_01_r}}]
    constraintWeights [1,1]
   }
   n28 = "Constraint - chain_01_r_DSP (PP)"(ComponentPose=@n27, Alpha=1, ConstraintWeights_0=0.7, ConstraintWeights_1=0.3) {
    alphaBoolBlend {blendOption=Linear}
    boneToModify {boneName=chain_01_r_DSP}
    constraintSetup [{targetBone {boneName=dyn_chain_front_01_r}}, {targetBone {boneName=dyn_chain_mid_01_r}}]
    constraintWeights [1,1]
   }
   n29 = "Constraint - chain_02_r_DSP (PP)"(ComponentPose=@n28, Alpha=1, ConstraintWeights_0=0.3, ConstraintWeights_1=0.7) {
    alphaBoolBlend {blendOption=Linear}
    boneToModify {boneName=chain_02_r_DSP}
    constraintSetup [{targetBone {boneName=dyn_chain_front_01_r}}, {targetBone {boneName=dyn_chain_mid_01_r}}]
    constraintWeights [1,1]
   }
   n30 = "Constraint - chain_06_r_DSP (PPPP)"!(
    ComponentPose=@n29
    Alpha=1
    ConstraintWeights_0=0
    ConstraintWeights_1:float=0
    ConstraintWeights_2:float=0.8
    ConstraintWeights_3:float=0.2
   ) -> (Pose) {
    alphaBoolBlend {blendOption=Linear}
    boneToModify {boneName=chain_06_r_DSP}
    constraintSetup [
     {targetBone {boneName=dyn_chain_back_01_r}}
     {targetBone {boneName=dyn_chain_mid_01_r}}
     {targetBone {boneName=chain_07_r_DSP}}
     {targetBone {boneName=chain_04_r_DSP}}
    ]
    constraintWeights [1,1,1,1]
   }
   n31 = "Constraint - chain_05_r_DSP (PPPP)"!(
    ComponentPose=@n30
    Alpha=1
    ConstraintWeights_0=0
    ConstraintWeights_1:float=0
    ConstraintWeights_2:float=0.2
    ConstraintWeights_3:float=0.8
   ) -> (Pose) {
    alphaBoolBlend {blendOption=Linear}
    boneToModify {boneName=chain_05_r_DSP}
    constraintSetup [
     {targetBone {boneName=dyn_chain_back_01_r}}
     {targetBone {boneName=dyn_chain_mid_01_r}}
     {targetBone {boneName=chain_06_r_DSP}}
     {targetBone {boneName=chain_04_r_DSP}}
    ]
    constraintWeights [1,1,1,1]
   }
   n32 = Get kneeSwingDelta_l() -> (kneeSwingDelta_l:double)
   n33 = "Constraint - kneepad_01_l (PP)"!(ComponentPose=@n31, Alpha=@n32, ConstraintWeights_0=0.25, ConstraintWeights_1:float=0.75) -> (Pose) {
    alphaBoolBlend {blendOption=Linear}
    boneToModify {boneName=kneepad_01_l}
    constraintSetup [{targetBone {boneName=thigh_l}}, {targetBone {boneName=calf_l}}]
    constraintWeights [1,1]
   }
   n34 = "Transform (Modify) Bone - Bone: kneepad_01_l"(ComponentPose=@n33, Alpha=@n32, Translation="-2,0,-2", Rotation="-15,0,0", Scale="1,1.2,1") {
    rotationMode=BMM_Additive
    rotationSpace=BCS_BoneSpace
    scaleMode=BMM_Additive
    scaleSpace=BCS_BoneSpace
    translationMode=BMM_Additive
    translationSpace=BCS_BoneSpace
    alphaBoolBlend {blendOption=Linear}
    boneToModify {boneName=kneepad_01_l}
    rotation {pitch=-15}
    scale {y=1.2}
    translation {x=-2, z=-2}
   }
   n35 = Get kneeSwingDelta_r() -> (kneeSwingDelta_r:double)
   n36 = "Constraint - kneepad_01_r (PP)"!(ComponentPose=@n34, Alpha=@n35, ConstraintWeights_0=0.25, ConstraintWeights_1:float=0.75) -> (Pose) {
    alphaBoolBlend {blendOption=Linear}
    boneToModify {boneName=kneepad_01_r}
    constraintSetup [{targetBone {boneName=thigh_r}}, {targetBone {boneName=calf_r}}]
    constraintWeights [1,1]
   }
   n37 = "Transform (Modify) Bone - Bone: kneepad_01_r"(ComponentPose=@n36, Alpha=@n35, Translation="1,0,2", Rotation="-15,0,0", Scale="1,1.2,1") {
    rotationMode=BMM_Additive
    rotationSpace=BCS_BoneSpace
    scaleMode=BMM_Additive
    scaleSpace=BCS_BoneSpace
    translationMode=BMM_Additive
    translationSpace=BCS_BoneSpace
    alphaBoolBlend {blendOption=Linear}
    boneToModify {boneName=kneepad_01_r}
    rotation {pitch=-15}
    scale {y=1.2}
    translation {x=1, z=2}
   }
   n38 = "Transform (Modify) Bone - Bone: shoulderpad_01_l"(ComponentPose=@n37, Alpha=Get armSwingDelta_L() -> (armSwingDelta_L:double), Translation="1.75,0,0", Rotation="0,0,0", Scale="1,1,1") {
    rotationMode=BMM_Additive
    rotationSpace=BCS_BoneSpace
    scaleMode=BMM_Additive
    scaleSpace=BCS_BoneSpace
    translationMode=BMM_Additive
    translationSpace=BCS_BoneSpace
    alphaBoolBlend {blendOption=Linear}
    boneToModify {boneName=shoulderpad_01_l}
    translation {x=1.75}
   }
   Output Pose(Result=Component To Local(ComponentPose=@n38)) {layerGroup=None}
   n39 = "Constraint - dyn_chain_front_01_l_MSTR (R)"(ComponentPose=Local To Component(LocalPose=Copy Pose From Mesh() {bUseAttachedParent=true}), Alpha=1, ConstraintWeights_0=1) {
    alphaBoolBlend {blendOption=Linear}
    boneToModify {boneName=dyn_chain_front_01_l_MSTR}
    constraintSetup [{transformType=Rotation, targetBone {boneName=thigh_twist_01_l}}]
    constraintWeights [1]
   }
   n40 = "Constraint - dyn_chain_mid_01_l_MSTR (R)"(ComponentPose=@n39, Alpha=1, ConstraintWeights_0=1) {
    alphaBoolBlend {blendOption=Linear}
    boneToModify {boneName=dyn_chain_mid_01_l_MSTR}
    constraintSetup [{transformType=Rotation, targetBone {boneName=thigh_twist_01_l}}]
    constraintWeights [1]
   }
   n41 = "Transform (Modify) Bone - Bone: dyn_chain_front_01_l_MSTR"(
    ComponentPose=@n40
    Alpha=Get isCrouchingAlpha() -> (isCrouchingAlpha:double)
    Translation="2,-1.000,0.000"
    Rotation="40.000,-35,-0.000"
    Scale="1,1,1"
   ) {
    rotationMode=BMM_Additive
    translationMode=BMM_Additive
    alphaBoolBlend {blendOption=Linear}
    boneToModify {boneName=dyn_chain_front_01_l_MSTR}
    rotation {pitch=40, yaw=-35}
    translation=<2 -1>
   }
   n42 = "Constraint - dyn_chain_front_01_r_MSTR (R)"(ComponentPose=@n41, Alpha=1, ConstraintWeights_0=1) {
    alphaBoolBlend {blendOption=Linear}
    boneToModify {boneName=dyn_chain_front_01_r_MSTR}
    constraintSetup [{transformType=Rotation, targetBone {boneName=thigh_twist_01_r}}]
    constraintWeights [1]
   }
   n43 = "Constraint - dyn_chain_mid_01_r_MSTR (R)"(ComponentPose=@n42, Alpha=1, ConstraintWeights_0=1) {
    alphaBoolBlend {blendOption=Linear}
    boneToModify {boneName=dyn_chain_mid_01_r_MSTR}
    constraintSetup [{transformType=Rotation, targetBone {boneName=thigh_twist_01_r}}]
    constraintWeights [1]
   }
   n44 = "Transform (Modify) Bone - Bone: dyn_chain_front_01_r_MSTR"(ComponentPose=@n43, Alpha=0, Translation="2,-1.000,0.000", Rotation="40.000,-35,-0.000", Scale="1,1,1") {
    rotationMode=BMM_Additive
    translationMode=BMM_Additive
    alphaBoolBlend {blendOption=Linear}
    boneToModify {boneName=dyn_chain_front_01_r_MSTR}
    rotation {pitch=40, yaw=-35}
    translation=<2 -1>
   }
   n45 = "Constraint - thigh_front_l_01_DSP (P)"(ComponentPose=@n44, Alpha=1, ConstraintWeights_0=1) {
    alphaBoolBlend {blendOption=Linear}
    boneToModify {boneName=thigh_front_l_01_DSP}
    constraintSetup [{targetBone {boneName=thigh_twist_01_l}}]
    constraintWeights [1]
   }
   n46 = Get thighSwingDelta_l() -> (thighSwingDelta_l:double)
   n47 = "Transform (Modify) Bone - Bone: thigh_front_l_01_DSP"(ComponentPose=@n45, Alpha=@n46, Translation="0.000,0.000,-10", Rotation="0,0,-20", Scale="1,1,1") {
    rotationMode=BMM_Additive
    scaleMode=BMM_Additive
    translationMode=BMM_Additive
    alphaBoolBlend {blendOption=Linear}
    boneToModify {boneName=thigh_front_l_01_DSP}
    rotation {roll=-20}
    translation {z=-10}
   }
   n48 = "Transform (Modify) Bone - Bone: thigh_back_l_01_DSP"(ComponentPose=@n47, Alpha=@n46, Translation="0,-5,-2", Rotation="0.000,0.000,45", Scale="1.25,1.000,1.000") {
    rotationMode=BMM_Additive
    scaleMode=BMM_Additive
    translationMode=BMM_Additive
    alphaBoolBlend {blendOption=Linear}
    boneToModify {boneName=thigh_back_l_01_DSP}
    rotation {roll=45}
    scale {x=1.25}
    translation {y=-5, z=-2}
   }
   n49 = "Transform (Modify) Bone - Bone: thigh_twist_01_l"(ComponentPose=@n48, Alpha=@n46, Translation="0,0,0", Rotation="0,0,0", Scale="1,1.25,1") {
    scaleMode=BMM_Additive
    translationMode=BMM_Additive
    translationSpace=BCS_BoneSpace
    alphaBoolBlend {blendOption=Linear}
    boneToModify {boneName=thigh_twist_01_l}
    scale {y=1.25}
   }
   n50 = "Constraint - thigh_front_r_01_DSP (P)"(ComponentPose=@n49, Alpha=1, ConstraintWeights_0=1) {
    alphaBoolBlend {blendOption=Linear}
    boneToModify {boneName=thigh_front_r_01_DSP}
    constraintSetup [{targetBone {boneName=thigh_twist_01_r}}]
    constraintWeights [1]
   }
   n51 = Get thighSwingDelta_r() -> (thighSwingDelta_r:double)
   n52 = "Transform (Modify) Bone - Bone: thigh_front_r_01_DSP"(ComponentPose=@n50, Alpha=@n51, Translation="0.000,0.000,-10", Rotation="0.000,0.000,10", Scale="1,1,1") {
    rotationMode=BMM_Additive
    scaleMode=BMM_Additive
    alphaBoolBlend {blendOption=Linear}
    boneToModify {boneName=thigh_front_r_01_DSP}
    rotation {roll=10}
    translation {z=-10}
   }
   n53 = "Transform (Modify) Bone - Bone: thigh_back_r_01_DSP"(ComponentPose=@n52, Alpha=@n51, Translation="0,-2,0", Rotation="0.000,0.000,75", Scale="1.25,1.000,1.000") {
    rotationMode=BMM_Additive
    scaleMode=BMM_Additive
    translationMode=BMM_Additive
    alphaBoolBlend {blendOption=Linear}
    boneToModify {boneName=thigh_back_r_01_DSP}
    rotation {roll=75}
    scale {x=1.25}
    translation {y=-2}
   }
   n54 = "Transform (Modify) Bone - Bone: thigh_twist_01_r"(ComponentPose=@n53, Alpha=@n51, Translation="0,0,0", Rotation="0,0,0", Scale="1,1.25,1") {
    scaleMode=BMM_Additive
    translationMode=BMM_Additive
    translationSpace=BCS_BoneSpace
    alphaBoolBlend {blendOption=Linear}
    boneToModify {boneName=thigh_twist_01_r}
    scale {y=1.25}
   }
   n55 = Get headTwistAngleAlpha_L() -> (headTwistAngleAlpha_L:double)
   n56 = "Transform (Modify) Bone - Bone: dyn_collar_front_01_l_MSTR"(ComponentPose=@n54, Alpha=@n55, Translation="0,0,0", Rotation="-35,0,0", Scale="1,1,1") {
    rotationMode=BMM_Additive
    rotationSpace=BCS_BoneSpace
    alphaBoolBlend {blendOption=Linear}
    boneToModify {boneName=dyn_collar_front_01_l_MSTR}
    rotation {pitch=-35}
   }
   n57 = "Transform (Modify) Bone - Bone: dyn_jacket_upper_01_l_MSTR"(ComponentPose=@n56, Alpha=@n55, Translation="0,0,0", Rotation="35,0,0", Scale="1,1,1") {
    rotationMode=BMM_Additive
    rotationSpace=BCS_BoneSpace
    alphaBoolBlend {blendOption=Linear}
    boneToModify {boneName=dyn_jacket_upper_01_l_MSTR}
    rotation {pitch=35}
   }
   n58 = Get headTwistAngleAlpha_R() -> (headTwistAngleAlpha_R:double)
   n59 = "Transform (Modify) Bone - Bone: dyn_collar_front_01_r_MSTR"(ComponentPose=@n57, Alpha=@n58, Translation="0,0,0", Rotation="-35,0,0", Scale="1,1,1") {
    rotationMode=BMM_Additive
    rotationSpace=BCS_BoneSpace
    alphaBoolBlend {blendOption=Linear}
    alphaScaleBiasClamp {bInterpResult=true}
    boneToModify {boneName=dyn_collar_front_01_r_MSTR}
    rotation {pitch=-35}
   }
   n60 = "Transform (Modify) Bone - Bone: dyn_jacket_upper_01_r_MSTR"(ComponentPose=@n59, Alpha=@n58, Translation="0,0,0", Rotation="35,0,0", Scale="1,1,1") {
    rotationMode=BMM_Additive
    rotationSpace=BCS_BoneSpace
    alphaBoolBlend {blendOption=Linear}
    alphaScaleBiasClamp {bInterpResult=true}
    boneToModify {boneName=dyn_jacket_upper_01_r_MSTR}
    rotation {pitch=35}
   }
   n61 = Get neckRot() -> (neckRot:Rotator)
   n62 = Get headSwingAngleAlpha() -> (headSwingAngleAlpha:double)
   n63 = "Transform (Modify) Bone - Bone: dyn_collar_back_01_MSTR"(ComponentPose=@n60, Alpha=@n62, Translation="0,0,0", Rotation=@n61, Scale="1,1,1") {rotationMode=BMM_Additive, rotationSpace=BCS_BoneSpace, alphaBoolBlend {blendOption=Linear}, boneToModify {boneName=dyn_collar_back_01_MSTR}}
   n64 = "Transform (Modify) Bone - Bone: dyn_collar_side_01_r_MSTR"(ComponentPose=@n63, Alpha=@n62, Translation="0,0,0", Rotation=@n61, Scale="1,1,1") {
    rotationMode=BMM_Additive
    rotationSpace=BCS_BoneSpace
    alphaBoolBlend {blendOption=Linear}
    boneToModify {boneName=dyn_collar_side_01_r_MSTR}
   }
   n65 = Get armSwingDelta_L()
   n66 = "Transform (Modify) Bone - Bone: dyn_collar_front_01_r_MSTR"(ComponentPose=@n64, Alpha=@n65, Translation="0,0,0", Rotation="25,0,0", Scale="1,1,1") {
    rotationMode=BMM_Additive
    rotationSpace=BCS_BoneSpace
    alphaBoolBlend {blendOption=Linear}
    boneToModify {boneName=dyn_collar_front_01_r_MSTR}
    rotation {pitch=25}
   }
   n67 = "Transform (Modify) Bone - Bone: dyn_jacket_upper_01_r_MSTR"(ComponentPose=@n66, Alpha=@n65, Translation="0,0,0", Rotation="-45,0,0", Scale="1,1,1") {
    rotationMode=BMM_Additive
    rotationSpace=BCS_BoneSpace
    alphaBoolBlend {blendOption=Linear}
    boneToModify {boneName=dyn_jacket_upper_01_r_MSTR}
    rotation {pitch=-45}
   }
   Save cached pose 'preSimCache'(Pose=Component To Local(ComponentPose=@n67)) {cacheName=preSimCache}
   Get isCrouchingAlpha()
  ]}
 EventGraph {nodes [
   Event BlueprintUpdateAnimation()
   > n1 = Sequence!(execute) -> (then_0, then_1, then_2:Exec)
   n2 = Get SkydivingExternalForceInput()
   > Set windVector(windVector:Vector=vector * float(
     A=Break Fort Anim Input Skydiving External Force(FortAnimInput_SkydivingExternalForce=@n2) -> (
      bUseSkydivingVectorForce=~
      SkydivingVectorForce=~
      HeadToPelvisDirection=~
      FloatingMultiplier=~
      FloatingAdditive=~
      DivingMultiplier=~
      DivingAdditive=~
      ParachutingMultiplier=~
      ParachutingAdditive=~
     ).SkydivingVectorForce
     B=3
    )) -> (Output_Get:Vector)
   n3 = GetOwnerSkeletalMeshComponent()
   n4 = BreakRotator(InRot=BreakTransform(InTransform=GetDeltaTransformFromRefPose(self=@n3, BoneName=calf_l, BaseName=thigh_l)).Rotation)
   Set kneeSwingDelta_l(execute=@n1.then_1, kneeSwingDelta_l:double=MapRangeClamped(Value="Absolute (Float)"(A=@n4.Yaw), InRangeA=50, InRangeB=110)) -> (Output_Get:double)
   n5 = BreakRotator(InRot=BreakTransform(InTransform=GetDeltaTransformFromRefPose(self=@n3, BoneName=calf_r, BaseName=thigh_r)).Rotation)
   > Set kneeSwingDelta_r(kneeSwingDelta_r:double=MapRangeClamped(Value="Absolute (Float)"(A=@n5.Yaw), InRangeA=50, InRangeB=110)) -> (Output_Get:double)
   n6 = BreakRotator(InRot=BreakTransform(InTransform=GetDeltaTransformFromRefPose(self=@n3, BoneName=thigh_l, BaseName=pelvis)).Rotation)
   > Set thighSwingDelta_l(thighSwingDelta_l:double=MapRangeClamped(Value="Absolute (Float)"(A=@n6.Yaw), InRangeA=50, InRangeB=110)) -> (Output_Get:double)
   n7 = GetDeltaTransformFromRefPose(self=@n3, BoneName=thigh_r, BaseName=pelvis)
   n8 = "Absolute (Float)"(A=BreakRotator(InRot=BreakTransform(InTransform=@n7).Rotation).Yaw)
   > Set thighSwingDelta_r(thighSwingDelta_r:double=MapRangeClamped(Value=@n8, InRangeA=50, InRangeB=110)) -> (Output_Get:double)
   n9 = Get bIsCrouching()
   > n10 = Set thighSwingDelta_r_post(thighSwingDelta_r_post:double=MapRangeClamped(Value=@n8, InRangeA=50, InRangeB=110, OutRangeA=SelectFloat(A=1, B=0, bPickA=@n9))) -> (Output_Get:double)
   > Set thighSwingDelta_r_post_mid_rotation(thighSwingDelta_r_post_mid_rotation:double=SelectFloat(A=1, B=0.5, bPickA=Get bIsCrouching())) -> (Output_Get:double)
   n11 = BreakRotator(InRot=BreakTransform(InTransform=GetDeltaTransformFromRefPose(self=@n3, BoneName=head, BaseName=neck_02)).Rotation)
   > n12 = Set headTwistAngleAlpha_L(headTwistAngleAlpha_L:double=MapRangeClamped(Value=@n11, InRangeA=-10, InRangeB=10)) -> (Output_Get:double)
   > Set headTwistAngleAlpha_R(headTwistAngleAlpha_R:double=MapRangeClamped(Value=float - float(B=@n12) {is=title K2Node_CallFunction}, InRangeA=0, InRangeB=1, OutRangeA=0.25)) -> (Output_Get:double)
   n13 = Get bIsTargeting()
   n14 = Get bIsCrouching()
   n15 = Get bIsAccelerating2D()
   n16 = SelectFloat(A=0.35, B=0.1, bPickA=OR Boolean(A=AND Boolean!(A=@n13, B=@n14, C:bool=@n15) -> (ReturnValue), B=Get bIsSkyDiving()))
   n17 = Get PawnSpeed2D()
   n18 = MapRangeClamped(Value=@n17, InRangeA=0, InRangeB=150, OutRangeA=1, OutRangeB=1.15)
   > Set headSwingAngleAlpha(headSwingAngleAlpha:double=MapRangeClamped(
     Value=GetTwistAndSwingAngleOfDeltaRotationFromRefPose(self=@n3, BoneName=head).OutSwingAngle
     InRangeA=10
     InRangeB=45
     OutRangeA=float * float(A=@n16, B=@n18)
    )) -> (Output_Get:double)
   n19 = Get isSkydiveNotFloating() -> (isSkydiveNotFloating:bool)
   n20 = Get bIsSkydiveFloating()
   n21 = SelectFloat(A=-55, B=-15, bPickA=OR Boolean(A=Get bIsCrouching(), B=OR Boolean(A=@n20, B=@n19)))
   n22 = MapRangeClamped(Value=@n17, InRangeA=0, InRangeB=150, OutRangeA=1, OutRangeB=1.5)
   > Set neckRot(neckRot:Rotator=MakeRotator(Pitch=float * float(A=@n22, B=@n21))) -> (Output_Get:Rotator)
   n23 = BreakRotator(InRot=BreakTransform(InTransform=GetDeltaTransformFromRefPose(self=@n3, BoneName=upperarm_l, BaseName=clavicle_l)).Rotation)
   n24 = Get bIsTargeting()
   > Set armSwingDelta_L(armSwingDelta_L:double=MapRangeClamped(
     Value="Absolute (Float)"(A=@n23.Yaw)
     InRangeA=45
     InRangeB=70
     OutRangeA=SelectFloat(A=0.5, B=0, bPickA=@n24)
     OutRangeB="To Float (Boolean)"(InBool=@n24)
    )) -> (Output_Get:double)
   n25 = Get bIsSkydiveFloating()
   n26 = Get bIsSkydiveDiveMode()
   Set isSkydiveNotFloating(execute=@n1.then_2, isSkydiveNotFloating:bool=AND Boolean(A=NOT Boolean(A=@n25), B=Get bIsSkyDiving(), C=NOT Boolean(A=@n26))) -> (Output_Get:bool)
   Event BlueprintInitializeAnimation()
   > Set isCrouchingAlpha(isCrouchingAlpha:double=0) -> (Output_Get:double)
   > Set isNotCrouchingAlpha(isNotCrouchingAlpha:double=1) -> (Output_Get:double)
   > ResetDynamics(InTeleportType=TeleportPhysics)
   > ResetDynamics(InTeleportType=ResetPhysics)
   Event OnBeginCrouching()
   > Set isCrouchingAlpha(isCrouchingAlpha=1)
   > Set isNotCrouchingAlpha(isNotCrouchingAlpha=0)
   > ResetDynamics(InTeleportType=TeleportPhysics)
   > ResetDynamics(InTeleportType=ResetPhysics)
   Event OnEndCrouching()
   > Set isCrouchingAlpha(isCrouchingAlpha=0)
   > Set isNotCrouchingAlpha(isNotCrouchingAlpha=1)
   > Delay(Duration=0.125)
   > ResetDynamics(InTeleportType=TeleportPhysics)
   > ResetDynamics(InTeleportType=ResetPhysics)
   TryGetPawnOwner()
   MapRangeClamped(Value=@n11, InRangeA=-10, InRangeB=5)
   n27 = NOT Boolean(A=@n9)
   n28 = "float >= float"(A="Absolute (Float)"(A=Get PawnVelocityZ()), B=1) {is=title K2Node_CallFunction}
   n29 = OR Boolean(A=Get bIsJumping(), B=Get bIsFalling())
   Set thighSwingDelta_r_post_translation(thighSwingDelta_r_post_translation:double=SelectFloat(A=float * float(A=@n10, B=1.1), B=0, bPickA=AND Boolean(A=@n27, B=@n28, C=@n29))) -> (Output_Get:double)
  ]}
]
variables [
 armSwingDelta_L {type=number}
 headSwingAngleAlpha {type=number}
 headTwistAngleAlpha_L {type=number}
 headTwistAngleAlpha_R {type=number}
 isCrouchingAlpha {type=number}
 isNotCrouchingAlpha {type=number}
 isSkydiveNotFloating {type=boolean}
 kneeSwingDelta_l {type=number}
 kneeSwingDelta_r {type=number}
 neckRot {type=Rotator}
 thighSwingDelta_l {type=number}
 thighSwingDelta_r {type=number}
 thighSwingDelta_r_post {type=number}
 thighSwingDelta_r_post_mid_rotation {type=number}
 thighSwingDelta_r_post_translation {type=number}
 windVector {type=Vector}
]
