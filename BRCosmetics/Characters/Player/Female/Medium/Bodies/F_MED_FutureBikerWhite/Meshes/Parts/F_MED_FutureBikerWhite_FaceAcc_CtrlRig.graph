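# Control Rig for the FutureBikerWhite female face accessory mesh that points its left and right ponytails at aim bones parented to spine_04.
# Each ponytail master bone takes a from-two-vectors quaternion delta on its X axis, sets global rotation and propagates to children.
# The right side reverses the direction vector for its mirrored axis.
# Head-pitch-driven SlideChain nodes sit unconnected to execution, and debug axis draws are disabled.
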ControlRigBlueprint /BRCosmetics/Characters/Player/Female/Medium/Bodies/F_MED_FutureBikerWhite/Meshes/Parts/F_MED_FutureBikerWhite_FaceAcc_CtrlRig
graphs [
 RigVMFunctionLibrary {links [], nodes []}
 RigVMModel {nodes [
   n1 = &RigUnit_BeginExecution()
   n2 = /Script/ControlRig.RigUnit_BoneName(Bone=dyn_ponytail_01_l_MSTR) {is=struct RigVMUnitNode}
   /Script/ControlRig.RigUnit_DebugTransformMutable(
    Color=(R=1,G=0,B=0,A=1)
    DebugDrawSettings=(DepthPriority=SDPG_Foreground,Lifetime=-1)
    Mode=Axes
    Scale=50
    Thickness=0
    Transform=(Rotation=(X=0,Y=0,Z=0,W=1),Translation=(X=0,Y=0,Z=0),Scale3D=(X=1,Y=1,Z=1))
    WorldOffset=(Rotation=(X=0,Y=0,Z=0,W=1),Translation=(X=0,Y=0,Z=0),Scale3D=(X=1,Y=1,Z=1))
    bEnabled=false
    ExecutePin=@n1.ExecutePin
    Space=@n2.Bone
   ) {is=struct RigVMUnitNode}
   n3 = &RigUnit_GetBoneTransform(Bone=@n2.Bone)
   n4 = /Script/RigVM.RigVMFunction_MathQuaternionGetAxis(Axis=X, Quaternion=@n3.Transform.Rotation) {is=struct RigVMUnitNode, template="Axis::Execute(in Quaternion,in Axis,out Result)"}
   n5 = /Script/RigVM.RigVMFunction_MathVectorSub(
    A=&RigUnit_GetBoneTransform(Bone=&RigUnit_BoneName(Bone=dyn_ponytail_aim_01_l).Bone).Transform.Translation
    B=@n3.Transform.Translation
   ) {is=struct RigVMUnitNode, template="Subtract::Execute(in A,in B,out Result)"}
   > /Script/ControlRig.RigUnit_SetBoneRotation(
    Space=GlobalSpace
    Weight=1
    bPropagateToChildren=True
    Bone=@n2.Bone
    Rotation=/Script/RigVM.RigVMFunction_MathQuaternionMul(
     B=@n3.Transform.Rotation
     A=/Script/RigVM.RigVMFunction_MathQuaternionFromTwoVectors(A=@n4.Result, B=@n5.Result) {is=struct RigVMUnitNode, template="FromTwoVectors::Execute(in A,in B,out Result)"}.Result
    ) {is=struct RigVMUnitNode, template="Multiply::Execute(in A,in B,out Result)"}.Result
   ) {is=struct RigVMUnitNode}
   > &RigUnit_DebugTransformMutable(
    Color=(R=1,G=0,B=0,A=1)
    DebugDrawSettings=(DepthPriority=SDPG_Foreground,Lifetime=-1)
    Mode=Axes
    Scale=75
    Thickness=0
    Transform=(Rotation=(X=0,Y=0,Z=0,W=1),Translation=(X=0,Y=0,Z=0),Scale3D=(X=1,Y=1,Z=1))
    WorldOffset=(Rotation=(X=0,Y=0,Z=0,W=1),Translation=(X=0,Y=0,Z=0),Scale3D=(X=1,Y=1,Z=1))
    bEnabled=false
    Space=@n2.Bone
   )
   n6 = &RigUnit_BoneName(Bone=dyn_ponytail_01_r_MSTR)
   > &RigUnit_DebugTransformMutable(
    Color=(R=1,G=0,B=0,A=1)
    DebugDrawSettings=(DepthPriority=SDPG_Foreground,Lifetime=-1)
    Mode=Axes
    Scale=50
    Thickness=0
    Transform=(Rotation=(X=0,Y=0,Z=0,W=1),Translation=(X=0,Y=0,Z=0),Scale3D=(X=1,Y=1,Z=1))
    WorldOffset=(Rotation=(X=0,Y=0,Z=0,W=1),Translation=(X=0,Y=0,Z=0),Scale3D=(X=1,Y=1,Z=1))
    bEnabled=false
    Space=@n6.Bone
   )
   n7 = &RigUnit_GetBoneTransform(Bone=@n6.Bone)
   n8 = &RigVMFunction_MathQuaternionGetAxis(Axis=X, Quaternion=@n7.Transform.Rotation) {template="Axis::Execute(in Quaternion,in Axis,out Result)"}
   n9 = &RigVMFunction_MathVectorSub(
    B=&RigUnit_GetBoneTransform(Bone=&RigUnit_BoneName(Bone=dyn_ponytail_aim_01_r).Bone).Transform.Translation
    A=@n7.Transform.Translation
   ) {template="Subtract::Execute(in A,in B,out Result)"}
   > &RigUnit_SetBoneRotation(
    Space=GlobalSpace
    Weight=1
    bPropagateToChildren=True
    Bone=@n6.Bone
    Rotation=&RigVMFunction_MathQuaternionMul(
     B=@n7.Transform.Rotation
     A=&RigVMFunction_MathQuaternionFromTwoVectors(A=@n8.Result, B=@n9.Result) {template="FromTwoVectors::Execute(in A,in B,out Result)"}.Result
    ) {template="Multiply::Execute(in A,in B,out Result)"}.Result
   )
   > &RigUnit_DebugTransformMutable(
    Color=(R=1,G=0,B=0,A=1)
    DebugDrawSettings=(DepthPriority=SDPG_Foreground,Lifetime=-1)
    Mode=Axes
    Scale=75
    Thickness=0
    Transform=(Rotation=(X=0,Y=0,Z=0,W=1),Translation=(X=0,Y=0,Z=0),Scale3D=(X=1,Y=1,Z=1))
    WorldOffset=(Rotation=(X=0,Y=0,Z=0,W=1),Translation=(X=0,Y=0,Z=0),Scale3D=(X=1,Y=1,Z=1))
    bEnabled=false
    Space=@n6.Bone
   )
   n10 = &RigVMFunction_MathFloatRemap(
    SourceMaximum=-50
    SourceMinimum=0
    TargetMaximum=0.1
    TargetMinimum=0
    bClamp=true
    Value=&RigVMFunction_MathQuaternionToEuler(RotationOrder=ZYX, Value=&RigUnit_GetRelativeBoneTransform(Bone=head, Space=spine_05).Transform.Rotation).Result.Z
   )
   /Script/ControlRig.RigUnit_SlideChain(
    ExecutePin=()
    bPropagateToChildren=True
    StartBone=@n2.Bone
    EndBone=&RigUnit_BoneName(Bone=dyn_ponytail_07_l).Bone
    SlideAmount=@n10.Result
   ) {is=struct RigVMUnitNode}
   &RigUnit_SlideChain(
    ExecutePin=()
    bPropagateToChildren=True
    StartBone=@n6.Bone
    EndBone=&RigUnit_BoneName(Bone=dyn_ponytail_07_r).Bone
    SlideAmount=@n10.Result
   )
  ]}
]
hierarchy [
 root {loc=<0 0 0>, rot=<0 0 0 1>}
 pelvis {parent=root, loc=<0 -0.5301 98.0571>, rot=<0.0127 0.707 -0.0127 -0.707>}
 spine_01 {parent=pelvis, loc=<6.1208 -0.9328 0>, rot=<0 0 0.034 -0.9994>}
 spine_02 {parent=spine_01, loc=<6.5996 -0.2095 0>, rot=<0 0 0.0238 -0.9997>}
 spine_03 {parent=spine_02, loc=<6.4914 -0.1237 0>, rot=<0 0 0.0149 -0.9999>}
 spine_04 {parent=spine_03, loc=<6.9598 0.0794 0>, rot=<0 0 -0.0132 -0.9999>}
 spine_05 {parent=spine_04, loc=<6.9987 0 0>, rot=<0 0 -0.0015 -1>}
 clavicle_l {parent=spine_05, loc=<5.4831 -0.1053 -3.2946>, rot=<0.0465 -0.7289 -0.009 -0.6829>}
 upperarm_l {parent=clavicle_l, loc=<9.8658 -1.2136 -0.4105>, rot=<0.0164 -0.3245 -0.0269 -0.9454>}
 lowerarm_l {parent=upperarm_l, loc=<26.465 -0.8345 0.259>, rot=<0.0036 0.0161 -0.2298 -0.9731>}
 clavicle_r {parent=spine_05, loc=<5.4831 -0.1053 3.2946>, rot=<0.7289 0.0465 -0.6829 0.009>}
 upperarm_r {parent=clavicle_r, loc=<-9.8658 1.2136 0.4104>, rot=<0.0164 -0.3245 -0.0269 -0.9454>}
 lowerarm_r {parent=upperarm_r, loc=<-26.4651 0.8345 -0.259>, rot=<0.0036 0.0161 -0.2298 -0.9731>}
 neck_01 {parent=spine_05, loc=<11.3138 0.1529 0>, rot=<0 0 -0.2003 -0.9797>}
 neck_02 {parent=neck_01, loc=<4.5627 0 0>, rot=<0 0 0.0504 -0.9987>}
 head {parent=neck_02, loc=<5.3914 0 0>, rot=<0 0 0.115 -0.9934>}
 faceAttach {parent=head, loc=<-3.5442 -0.4617 -0.0134>, rot=<0.0082 -0.7071 -0.0082 -0.7071>, scale=<0.89 0.89 0.89>}
 L_cheek_inner {parent=faceAttach, loc=<4.4352 8.4145 3.1576>, rot=<-0.0455 -0.2061 0.3265 -0.9213>}
 C_lip_upper_mid {parent=faceAttach, loc=<-0.023 10.2197 1.2402>, rot=<0.2225 0.0014 -0.0005 -0.9749>}
 L_lip_upper_outer {parent=faceAttach, loc=<1.3287 9.3628 1.1488>, rot=<0.0359 -0.0729 0.2627 -0.9614>}
 L_lip_corner {parent=faceAttach, loc=<2.2147 8.844 1.1198>, rot=<0.0033 0.0076 -0.0147 -0.9999>}
 C_nose_bridge {parent=faceAttach, loc=<-0.0368 11.2897 5.5674>, rot=<-0.2218 -0.0155 -0.0043 -0.975>}
 R_cheek_inner {parent=faceAttach, loc=<-4.4532 8.4232 3.1498>, rot=<-0.046 0.2075 -0.3264 -0.9211>}
 R_lip_corner {parent=faceAttach, loc=<-2.3008 8.7654 1.0868>, rot=<0.003 -0.0079 0.0146 -0.9999>}
 C_jaw {parent=faceAttach, loc=<-0.0407 3.0671 1.1699>, rot=<0 0 0 1>}
 L_lip_lower_outer {parent=C_jaw, loc=<1.3494 6.2216 -0.0927>, rot=<0.0022 -0.014 0.2607 -0.9653>}
 C_lip_lower_mid {parent=C_jaw, loc=<0.0258 6.923 -0.3352>, rot=<0 -0.0225 -0.0048 -0.9997>}
 R_lip_lower_outer {parent=C_jaw, loc=<-1.3495 6.4509 -0.2156>, rot=<-0.008 0.0693 -0.2166 -0.9738>}
 R_eye {parent=faceAttach, loc=<-3.2763 6.9812 7.9429>, rot=<0 0 0.9998 -0.0187>}
 R_eye_lid_lower_mid {parent=R_eye, loc=<0.0515 -0.3971 -1.0305>, rot=<0.0329 -0.0005 0.0167 -0.9993>}
 L_eye {parent=faceAttach, loc=<3.246 6.9812 7.9429>, rot=<0 0 0.9998 0.0187>}
 L_eye_lid_lower_mid {parent=L_eye, loc=<0.1623 -0.4796 -1.078>, rot=<0.0329 0 0 -0.9995>}
 R_lip_upper_outer {parent=faceAttach, loc=<-1.3728 9.393 1.2024>, rot=<0.0378 0.0676 -0.2607 -0.9623>}
 dyn_scalp_back_01 {parent=head, loc=<11.7599 -5.5401 0.4703>, rot=<0.6979 -0.2338 0.2621 -0.6242>}
 dyn_hairtopB_01_r_MSTR {parent=dyn_scalp_back_01, loc=<1.203 -2.8296 4.7006>, rot=<0.7753 -0.0012 0.6204 0.118>}
 dyn_hairtopB_01_r {parent=dyn_hairtopB_01_r_MSTR, loc=<0.2 0 0>, rot=<0 0 0 1>}
 dyn_scalp_front_01 {parent=head, loc=<10.1546 6.2645 -2.8047>, rot=<-0.3986 -0.4298 -0.2231 -0.7788>}
 dyn_hairfrontA_01_r_MSTR {parent=dyn_scalp_front_01, loc=<1.376 2.9165 -0.6788>, rot=<-0.0251 -0.5759 -0.6059 -0.5482>}
 dyn_hairfrontA_01_r {parent=dyn_hairfrontA_01_r_MSTR, loc=<0.2 0 0>, rot=<0 0 0 1>}
 dyn_hairfrontA_02_r {parent=dyn_hairfrontA_01_r, loc=<4.8039 0 0>, rot=<0 -0.3067 0 -0.9518>}
 dyn_hairfrontA_03_r {parent=dyn_hairfrontA_02_r, loc=<2.7296 0 0>, rot=<-0.0043 -0.1956 -0.0824 -0.9772>}
 dyn_hairfrontB_01_r_MSTR {parent=dyn_scalp_front_01, loc=<-0.2004 5.0354 1.5488>, rot=<0.0647 -0.4103 -0.7462 -0.5202>}
 dyn_hairfrontB_01_r {parent=dyn_hairfrontB_01_r_MSTR, loc=<0.2 0 0>, rot=<0 0 0 -1>}
 dyn_hairfrontB_02_r {parent=dyn_hairfrontB_01_r, loc=<4.78 0 0>, rot=<0 -0.3711 0 -0.9286>}
 dyn_hairfrontB_03_r {parent=dyn_hairfrontB_02_r, loc=<4.8855 0 0>, rot=<0 0 0 -1>}
 dyn_hairfrontC_01_r {parent=dyn_hairfrontB_01_r, loc=<5.2569 3.0114 1.2927>, rot=<0.1806 -0.3115 -0.1282 -0.9241>}
 dyn_hairfrontC_02_r {parent=dyn_hairfrontC_01_r, loc=<4.2992 0 0>, rot=<-0.1437 -0.0731 0.0005 -0.9869>}
 dyn_hairfrontD_01_r_MSTR {parent=dyn_scalp_front_01, loc=<-5.8864 8.4266 3.3516>, rot=<-0.2156 0.3857 0.8354 0.327>}
 dyn_hairfrontD_01_r {parent=dyn_hairfrontD_01_r_MSTR, loc=<0.2 0 0>, rot=<0 0 0 1>}
 dyn_hairfrontD_02_r {parent=dyn_hairfrontD_01_r, loc=<4.3756 0 0>, rot=<0 -0.1434 0 -0.9897>}
 dyn_hairfrontE_01_r_MSTR {parent=dyn_scalp_front_01, loc=<-6.2425 6.2353 7.1149>, rot=<0.1616 0.11 0.9431 0.2691>}
 dyn_hairfrontE_01_r {parent=dyn_hairfrontE_01_r_MSTR, loc=<0.2 0 0>, rot=<0 0 0 1>}
 dyn_hairfrontE_02_r {parent=dyn_hairfrontE_01_r, loc=<4.9663 0 0>, rot=<0 0.4781 0 -0.8783>}
 dyn_hairtop_01_r_MSTR {parent=dyn_scalp_front_01, loc=<-1.0867 0.59 9.6694>, rot=<0.1341 -0.1363 0.9256 0.3266>}
 dyn_hairtop_01_r {parent=dyn_hairtop_01_r_MSTR, loc=<0.2 0 0>, rot=<0 0 0 -1>}
 dyn_hairtop_02_r {parent=dyn_hairtop_01_r, loc=<4.3757 0 0>, rot=<0 -0.3123 0 -0.95>}
 dyn_hairtop_01_l_MSTR {parent=dyn_scalp_front_01, loc=<0.7743 -1.3035 6.6093>, rot=<0.7571 -0.131 0.566 0.299>}
 dyn_hairtop_01_l {parent=dyn_hairtop_01_l_MSTR, loc=<0.2 0 0>, rot=<0 0 0 1>}
 dyn_hairsideB_01_l_MSTR {parent=head, loc=<8.6495 1.2626 -5.516>, rot=<-0.0732 -0.9709 -0.0781 -0.2142>}
 dyn_hairsideB_01_l {parent=dyn_hairsideB_01_l_MSTR, loc=<0.2 0 0>, rot=<0 0 0 1>}
 dyn_hairsideB_02_l {parent=dyn_hairsideB_01_l, loc=<7.2079 0 0>, rot=<0 0 0 1>}
 dyn_hairside_01_l_MSTR {parent=head, loc=<8.7604 2.5166 -5.1733>, rot=<-0.3237 -0.8266 -0.4174 -0.1944>}
 dyn_hairside_01_l {parent=dyn_hairside_01_l_MSTR, loc=<0.2 0 0>, rot=<0 0 0 1>}
 dyn_hairside_02_l {parent=dyn_hairside_01_l, loc=<6.0411 0 0>, rot=<0.0049 -0.2056 -0.001 -0.9786>}
 dyn_hairside_03_l {parent=dyn_hairside_02_l, loc=<2.383 0 0>, rot=<0 -0.1223 0 -0.9925>}
 dyn_hairside_04_l {parent=dyn_hairside_03_l, loc=<1.5282 0 0>, rot=<0 0 0 -1>}
 dyn_ponytail_01_l_MSTR {parent=head, loc=<4.4283 0.7424 -6.7425>, rot=<-0.2072 -0.6966 -0.6553 -0.206>}
 dyn_ponytail_01_l {parent=dyn_ponytail_01_l_MSTR, loc=<0.2 0 0>, rot=<0 0 0 1>}
 dyn_ponytail_02_l {parent=dyn_ponytail_01_l, loc=<6.4897 0 0>, rot=<0.017 0 0 -0.9999>}
 dyn_ponytail_03_l {parent=dyn_ponytail_02_l, loc=<6.4897 0 0>, rot=<0 0 0 1>}
 dyn_ponytail_04_l {parent=dyn_ponytail_03_l, loc=<6.4896 0 0>, rot=<0 0 0 1>}
 dyn_ponytail_05_l {parent=dyn_ponytail_04_l, loc=<6.4897 0 0>, rot=<0 0 0 1>}
 dyn_ponytail_06_l {parent=dyn_ponytail_05_l, loc=<6.4896 0 0>, rot=<0 0 0 1>}
 dyn_ponytail_07_l {parent=dyn_ponytail_06_l, loc=<4.4418 0 0>, rot=<-0.0422 0.6424 0.1583 -0.7487>}
 dyn_ponytail_08_l {parent=dyn_ponytail_07_l, loc=<3.4316 0 0>, rot=<-0.0812 -0.1422 0.093 -0.9821>}
 dyn_ponytail_09_l {parent=dyn_ponytail_07_l, loc=<4.1926 1.3342 -3.1778>, rot=<-0.1312 -0.0583 -0.0904 -0.9855>}
 dyn_ponytail_010_l {parent=dyn_ponytail_07_l, loc=<1.2672 -0.5224 -3.678>, rot=<0.1917 -0.0143 0.1753 -0.9656>}
 zipper_01_l_MSTR {parent=head, loc=<-1.1884 8.3653 -4.1191>, rot=<0.0021 -0.3782 0.924 -0.0567>}
 zipper_01_l {parent=zipper_01_l_MSTR, loc=<0.2 0 0>, rot=<0 0 0 -1>}
 dyn_ponytail_01_r_MSTR {parent=head, loc=<4.4288 0.7424 6.7425>, rot=<-0.6966 0.2072 0.206 -0.6553>}
 dyn_ponytail_01_r {parent=dyn_ponytail_01_r_MSTR, loc=<-0.2001 0 0>, rot=<0 0 0 1>}
 dyn_ponytail_02_r {parent=dyn_ponytail_01_r, loc=<-6.4901 0 0.0003>, rot=<0.017 0 0 -0.9999>}
 dyn_ponytail_03_r {parent=dyn_ponytail_02_r, loc=<-6.4893 0 -0.0002>, rot=<0 0 0 1>}
 dyn_ponytail_04_r {parent=dyn_ponytail_03_r, loc=<-6.4901 0 0.0003>, rot=<0 0 0 1>}
 dyn_ponytail_05_r {parent=dyn_ponytail_04_r, loc=<-6.4893 0 -0.0003>, rot=<0 0 0 1>}
 dyn_ponytail_06_r {parent=dyn_ponytail_05_r, loc=<-6.4901 0 0.0004>, rot=<0 0 0 1>}
 dyn_ponytail_07_r {parent=dyn_ponytail_06_r, loc=<-4.4412 0 -0.0004>, rot=<-0.0422 0.6424 0.1583 -0.7487>}
 dyn_ponytail_08_r {parent=dyn_ponytail_07_r, loc=<-3.4313 0 0.0007>, rot=<-0.0812 -0.1422 0.093 -0.9821>}
 dyn_ponytail_09_r {parent=dyn_ponytail_07_r, loc=<-4.1924 -1.3342 3.1784>, rot=<-0.1312 -0.0583 -0.0904 -0.9855>}
 dyn_ponytail_010_r {parent=dyn_ponytail_07_r, loc=<-1.2673 0.5224 3.6779>, rot=<0.1917 -0.0143 0.1753 -0.9656>}
 dyn_ponytail_aim_01_r {parent=spine_04, loc=<1.0281 14.9309 5.7407>, rot=<-0.7048 0.1241 0.0997 -0.6913>}
 dyn_ponytail_aim_01_l {parent=spine_04, loc=<1.0281 14.9309 -5.7407>, rot=<-0.1241 -0.7048 -0.6913 -0.0997>}
]
