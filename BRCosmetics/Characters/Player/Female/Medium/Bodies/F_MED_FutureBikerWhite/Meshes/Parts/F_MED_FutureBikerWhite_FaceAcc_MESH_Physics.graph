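# Physics asset for the FutureBikerWhite female body's face accessory mesh, set up for two dangling ponytail chains.
# Left and right six-capsule chains hang from tiny MSTR anchor boxes linked to the head, with tight limits at the root and wider swing and twist down the chain.
# Capsules on head, neck, upper spine, clavicles and upper arms form body shapes joined by free-rotation constraints.
# A tiny zipper anchor box also links to the head.
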PhysicsAsset /BRCosmetics/Characters/Player/Female/Medium/Bodies/F_MED_FutureBikerWhite/Meshes/Parts/F_MED_FutureBikerWhite_FaceAcc_MESH_Physics
bodies {
 clavicle_l [{shapeType=Capsule, length=6.755, radius=9.505, center=<7.2033 0 0>, rotation=<rot 0 -89.9998 89.9994>}]
 clavicle_r [{shapeType=Capsule, length=6.755, radius=9.505, center=<-7.2033 0 0>, rotation=<rot 0 -89.9998 -90.0006>}]
 dyn_ponytail_01_l [{shapeType=Capsule, length=2, radius=2.5, center=<3.1002 0.0002 0>, rotation=<rot 90 -0.0033 -0.0033>}]
 dyn_ponytail_01_l_MSTR [{shapeType=Box, extentX=0.01, extentY=0.01, extentZ=0.01}]
 dyn_ponytail_01_r [{shapeType=Capsule, length=2, radius=2.5, center=<-3.1002 -0.0002 0>, rotation=<rot 90 0.005 -179.995>}]
 dyn_ponytail_01_r_MSTR [{shapeType=Box, extentX=0.01, extentY=0.01, extentZ=0.01}]
 dyn_ponytail_02_l [{shapeType=Capsule, length=2, radius=2.5, center=<3.1002 0.0002 0>, rotation=<rot 90 -0.0033 -0.0033>}]
 dyn_ponytail_02_r [{shapeType=Capsule, length=2, radius=2.5, center=<-3.1002 -0.0002 0>, rotation=<rot 90 0.005 -179.995>}]
 dyn_ponytail_03_l [{shapeType=Capsule, length=2, radius=2.5, center=<3.1002 0.0002 0>, rotation=<rot 90 -0.0033 -0.0033>}]
 dyn_ponytail_03_r [{shapeType=Capsule, length=2, radius=2.5, center=<-3.1002 -0.0002 0>, rotation=<rot 90 0.005 -179.995>}]
 dyn_ponytail_04_l [{shapeType=Capsule, length=2, radius=2.5, center=<3.1002 0.0002 0>, rotation=<rot 90 -0.0033 -0.0033>}]
 dyn_ponytail_04_r [{shapeType=Capsule, length=2, radius=2.5, center=<-3.1002 -0.0002 0>, rotation=<rot 90 0.005 -179.995>}]
 dyn_ponytail_05_l [{shapeType=Capsule, length=2, radius=2.5, center=<3.1002 0.0002 0>, rotation=<rot 90 -0.0033 -0.0033>}]
 dyn_ponytail_05_r [{shapeType=Capsule, length=2, radius=2.5, center=<-3.1002 -0.0002 0>, rotation=<rot 90 0.005 -179.995>}]
 dyn_ponytail_06_l [{shapeType=Capsule, length=2, radius=2.5, center=<3.1002 0.0002 0>, rotation=<rot 90 -0.0033 -0.0033>}]
 dyn_ponytail_06_r [{shapeType=Capsule, length=2, radius=2.5, center=<-3.1002 -0.0002 0>, rotation=<rot 90 0.005 -179.995>}]
 head [{shapeType=Capsule, length=9.25, radius=4.25, center=<1.6736 5.5864 0.0001>, rotation=<rot -90 14.0363 -4.0352>}]
 neck_02 [{shapeType=Capsule, length=3, radius=4.75, center=<-0.3267 1.8525 0.0002>, rotation=<rot -90 -176.4237 -173.5752>}]
 spine_04 [{shapeType=Capsule, length=6.5, radius=12, center=<0 3.6535 0>}]
 spine_05 [{shapeType=Capsule, length=8.25, radius=13.75, rotation=<rot 0 0.0014 0>}]
 upperarm_l [{shapeType=Capsule, length=21.505, radius=6.255, center=<13.7476 0.0002 0>, rotation=<rot 0 -89.9998 89.9994>}]
 upperarm_r [{shapeType=Capsule, length=21.505, radius=6.255, center=<-13.7476 -0.0002 0>, rotation=<rot 0 -89.9998 -90.0006>}]
 zipper_01_l_MSTR [{shapeType=Box, extentX=0.01, extentY=0.01, extentZ=0.01}]
}
constraints [
 {
  bone1Name=dyn_ponytail_01_l_MSTR
  bone2Name=head
  swing1LimitDegrees=45
  swing1Motion=Free
  swing2LimitDegrees=45
  swing2Motion=Free
  twistLimitDegrees=45
  twistMotion=Free
 }
 {
  bone1Name=dyn_ponytail_01_r_MSTR
  bone2Name=head
  swing1LimitDegrees=45
  swing1Motion=Free
  swing2LimitDegrees=45
  swing2Motion=Free
  twistLimitDegrees=45
  twistMotion=Free
 }
 {
  bone1Name=zipper_01_l_MSTR
  bone2Name=head
  swing1LimitDegrees=45
  swing1Motion=Free
  swing2LimitDegrees=45
  swing2Motion=Free
  twistLimitDegrees=45
  twistMotion=Free
 }
 {
  bone1Name=neck_02
  bone2Name=spine_05
  swing1LimitDegrees=45
  swing1Motion=Free
  swing2LimitDegrees=45
  swing2Motion=Free
  twistLimitDegrees=45
  twistMotion=Free
 }
 {
  bone1Name=head
  bone2Name=neck_02
  swing1LimitDegrees=45
  swing1Motion=Free
  swing2LimitDegrees=45
  swing2Motion=Free
  twistLimitDegrees=45
  twistMotion=Free
 }
 {
  bone1Name=spine_05
  bone2Name=spine_04
  swing1LimitDegrees=45
  swing1Motion=Free
  swing2LimitDegrees=45
  swing2Motion=Free
  twistLimitDegrees=45
  twistMotion=Free
 }
 {bone1Name=dyn_ponytail_01_r, bone2Name=dyn_ponytail_01_r_MSTR, swing1LimitDegrees=5, swing2LimitDegrees=15, twistLimitDegrees=5}
 {bone1Name=dyn_ponytail_06_r, bone2Name=dyn_ponytail_05_r, swing1LimitDegrees=40, swing2LimitDegrees=80, twistLimitDegrees=40}
 {bone1Name=dyn_ponytail_05_r, bone2Name=dyn_ponytail_04_r, swing1LimitDegrees=40, swing2LimitDegrees=80, twistLimitDegrees=40}
 {bone1Name=dyn_ponytail_04_r, bone2Name=dyn_ponytail_03_r, swing1LimitDegrees=40, swing2LimitDegrees=80, twistLimitDegrees=40}
 {bone1Name=dyn_ponytail_03_r, bone2Name=dyn_ponytail_02_r, swing1LimitDegrees=40, swing2LimitDegrees=80, twistLimitDegrees=40}
 {bone1Name=dyn_ponytail_02_r, bone2Name=dyn_ponytail_01_r, swing1LimitDegrees=40, swing2LimitDegrees=80, twistLimitDegrees=40}
 {bone1Name=dyn_ponytail_01_l, bone2Name=dyn_ponytail_01_l_MSTR, swing1LimitDegrees=5, swing2LimitDegrees=15, twistLimitDegrees=5}
 {bone1Name=dyn_ponytail_02_l, bone2Name=dyn_ponytail_01_l, swing1LimitDegrees=40, swing2LimitDegrees=80, twistLimitDegrees=40}
 {bone1Name=dyn_ponytail_03_l, bone2Name=dyn_ponytail_02_l, swing1LimitDegrees=40, swing2LimitDegrees=80, twistLimitDegrees=40}
 {bone1Name=dyn_ponytail_04_l, bone2Name=dyn_ponytail_03_l, swing1LimitDegrees=40, swing2LimitDegrees=80, twistLimitDegrees=40}
 {bone1Name=dyn_ponytail_05_l, bone2Name=dyn_ponytail_04_l, swing1LimitDegrees=40, swing2LimitDegrees=80, twistLimitDegrees=40}
 {bone1Name=dyn_ponytail_06_l, bone2Name=dyn_ponytail_05_l, swing1LimitDegrees=40, swing2LimitDegrees=80, twistLimitDegrees=40}
 {
  bone1Name=clavicle_l
  bone2Name=spine_05
  swing1LimitDegrees=45
  swing1Motion=Free
  swing2LimitDegrees=45
  swing2Motion=Free
  twistLimitDegrees=45
  twistMotion=Free
 }
 {
  bone1Name=clavicle_r
  bone2Name=spine_05
  swing1LimitDegrees=45
  swing1Motion=Free
  swing2LimitDegrees=45
  swing2Motion=Free
  twistLimitDegrees=45
  twistMotion=Free
 }
 {
  bone1Name=upperarm_l
  bone2Name=clavicle_l
  swing1LimitDegrees=45
  swing1Motion=Free
  swing2LimitDegrees=45
  swing2Motion=Free
  twistLimitDegrees=45
  twistMotion=Free
 }
 {
  bone1Name=upperarm_r
  bone2Name=clavicle_r
  swing1LimitDegrees=45
  swing1Motion=Free
  swing2LimitDegrees=45
  swing2Motion=Free
  twistLimitDegrees=45
  twistMotion=Free
 }
]
