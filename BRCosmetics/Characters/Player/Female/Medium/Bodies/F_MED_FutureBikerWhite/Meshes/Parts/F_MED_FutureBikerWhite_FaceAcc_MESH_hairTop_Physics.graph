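# Physics asset for the hair-top mesh of the FutureBikerWhite female character, covering dynamic hair strands, a zipper, head and neck.
# Box bodies form short front, side and top hair chains with tight twist limits, rooted to the head through tiny MSTR anchor boxes.
# The frontC chain branches from frontB's first segment.
# Head and neck capsules link with free rotation, as does the zipper's anchor, while the zipper itself is limited.
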PhysicsAsset /BRCosmetics/Characters/Player/Female/Medium/Bodies/F_MED_FutureBikerWhite/Meshes/Parts/F_MED_FutureBikerWhite_FaceAcc_MESH_hairTop_Physics
bodies {
 dyn_hairfrontA_01_r [{shapeType=Box, extentX=4, extentY=2, extentZ=1.6, center=<2.2614 0 0>}]
 dyn_hairfrontA_01_r_MSTR [{shapeType=Box, extentX=0.01, extentY=0.01, extentZ=0.01}]
 dyn_hairfrontA_02_r [{shapeType=Box, extentX=4, extentY=2, extentZ=1.6, center=<2.2614 0 0>}]
 dyn_hairfrontA_03_r [{shapeType=Box, extentX=4, extentY=2, extentZ=1.6, center=<2.2614 0 0>}]
 dyn_hairfrontB_01_r [{shapeType=Box, extentX=4, extentY=2, extentZ=1.6, center=<2.2614 0 0>}]
 dyn_hairfrontB_01_r_MSTR [{shapeType=Box, extentX=0.01, extentY=0.01, extentZ=0.01}]
 dyn_hairfrontB_02_r [{shapeType=Box, extentX=4, extentY=2, extentZ=1.6, center=<2.2614 0 0>}]
 dyn_hairfrontB_03_r [{shapeType=Box, extentX=4, extentY=2, extentZ=1.6, center=<2.2614 0 0>}]
 dyn_hairfrontC_01_r [{shapeType=Box, extentX=4, extentY=2, extentZ=1.6, center=<2.2614 0 0>}]
 dyn_hairfrontC_02_r [{shapeType=Box, extentX=4, extentY=2, extentZ=1.6, center=<2.2614 0 0>}]
 dyn_hairfrontD_01_r [{shapeType=Box, extentX=4, extentY=2, extentZ=1.6, center=<2.2614 0 0>}]
 dyn_hairfrontD_01_r_MSTR [{shapeType=Box, extentX=0.01, extentY=0.01, extentZ=0.01}]
 dyn_hairfrontD_02_r [{shapeType=Box, extentX=4, extentY=2, extentZ=1.6, center=<2.2614 0 0>}]
 dyn_hairfrontE_01_r [{shapeType=Box, extentX=4, extentY=2, extentZ=1.6, center=<2.2614 0 0>}]
 dyn_hairfrontE_01_r_MSTR [{shapeType=Box, extentX=0.01, extentY=0.01, extentZ=0.01}]
 dyn_hairfrontE_02_r [{shapeType=Box, extentX=4, extentY=2, extentZ=1.6, center=<2.2614 0 0>}]
 dyn_hairsideB_01_l [{shapeType=Box, extentX=4, extentY=2, extentZ=1.6, center=<2.2614 0 0>}]
 dyn_hairsideB_01_l_MSTR [{shapeType=Box, extentX=0.01, extentY=0.01, extentZ=0.01}]
 dyn_hairsideB_02_l [{shapeType=Box, extentX=4, extentY=2, extentZ=1.6, center=<2.2614 0 0>}]
 dyn_hairside_01_l [{shapeType=Box, extentX=4, extentY=2, extentZ=1.6, center=<2.2614 0 0>}]
 dyn_hairside_01_l_MSTR [{shapeType=Box, extentX=0.01, extentY=0.01, extentZ=0.01}]
 dyn_hairside_02_l [{shapeType=Box, extentX=4, extentY=2, extentZ=1.6, center=<2.2614 0 0>}]
 dyn_hairside_03_l [{shapeType=Box, extentX=4, extentY=2, extentZ=1.6, center=<2.2614 0 0>}]
 dyn_hairside_04_l [{shapeType=Box, extentX=4, extentY=2, extentZ=1.6, center=<2.2614 0 0>}]
 dyn_hairtopB_01_r [{shapeType=Box, extentX=4, extentY=2, extentZ=1.6, center=<2.2614 0 0>}]
 dyn_hairtopB_01_r_MSTR [{shapeType=Box, extentX=0.01, extentY=0.01, extentZ=0.01}]
 dyn_hairtop_01_l [{shapeType=Box, extentX=4, extentY=2, extentZ=1.6, center=<2.2614 0 0>}]
 dyn_hairtop_01_l_MSTR [{shapeType=Box, extentX=0.01, extentY=0.01, extentZ=0.01}]
 dyn_hairtop_01_r [{shapeType=Box, extentX=4, extentY=2, extentZ=1.6, center=<2.2614 0 0>}]
 dyn_hairtop_01_r_MSTR [{shapeType=Box, extentX=0.01, extentY=0.01, extentZ=0.01}]
 dyn_hairtop_02_r [{shapeType=Box, extentX=4, extentY=2, extentZ=1.6, center=<2.2614 0 0>}]
 head [{shapeType=Capsule, length=9.25, radius=4.25, center=<1.6071 5.9633 0.0001>, rotation=<rot -90 -176.4237 -173.5752>}]
 neck_01 [{shapeType=Capsule, length=1.25, radius=6.5, center=<-0.1384 1.876 0.0002>, rotation=<rot -90 -176.4237 -173.5752>}]
 neck_02 [{shapeType=Capsule, length=1.5, radius=6.25, center=<-0.3267 1.8525 0.0002>, rotation=<rot -90 -176.4237 -173.5752>}]
 zipper_01_l [{shapeType=Box, extentX=4, extentY=2, extentZ=1.6, center=<2.2614 0 0>}]
 zipper_01_l_MSTR [{shapeType=Box, extentX=0.01, extentY=0.01, extentZ=0.01}]
}
constraints [
 {bone1Name=dyn_hairtopB_01_r_MSTR, bone2Name=head, swing1LimitDegrees=45, swing2LimitDegrees=45, twistLimitDegrees=5}
 {bone1Name=dyn_hairfrontA_01_r_MSTR, bone2Name=head, swing1LimitDegrees=45, swing2LimitDegrees=45, twistLimitDegrees=5}
 {bone1Name=dyn_hairfrontB_01_r_MSTR, bone2Name=head, swing1LimitDegrees=45, swing2LimitDegrees=45, twistLimitDegrees=5}
 {bone1Name=dyn_hairfrontD_01_r_MSTR, bone2Name=head, swing1LimitDegrees=45, swing2LimitDegrees=45, twistLimitDegrees=5}
 {bone1Name=dyn_hairfrontE_01_r_MSTR, bone2Name=head, swing1LimitDegrees=45, swing2LimitDegrees=45, twistLimitDegrees=5}
 {bone1Name=dyn_hairtop_01_r_MSTR, bone2Name=head, swing1LimitDegrees=45, swing2LimitDegrees=45, twistLimitDegrees=5}
 {bone1Name=dyn_hairtop_01_l_MSTR, bone2Name=head, swing1LimitDegrees=45, swing2LimitDegrees=45, twistLimitDegrees=5}
 {bone1Name=dyn_hairsideB_01_l_MSTR, bone2Name=head, swing1LimitDegrees=45, swing2LimitDegrees=45, twistLimitDegrees=5}
 {bone1Name=dyn_hairside_01_l_MSTR, bone2Name=head, swing1LimitDegrees=45, swing2LimitDegrees=45, twistLimitDegrees=5}
 {
  bone1Name=zipper_01_l_MSTR
  bone2Name=head
  swing1LimitDegrees=45
  swing1Motion=Free
  swing2LimitDegrees=45
  swing2Motion=Free
  twistLimitDegrees=45
  twistMotion=Free
 }
 {
  bone1Name=neck_02
  bone2Name=neck_01
  swing1LimitDegrees=45
  swing1Motion=Free
  swing2LimitDegrees=45
  swing2Motion=Free
  twistLimitDegrees=45
  twistMotion=Free
 }
 {
  bone1Name=head
  bone2Name=neck_02
  swing1LimitDegrees=45
  swing1Motion=Free
  swing2LimitDegrees=45
  swing2Motion=Free
  twistLimitDegrees=45
  twistMotion=Free
 }
 {bone1Name=dyn_hairtopB_01_r, bone2Name=dyn_hairtopB_01_r_MSTR, swing1LimitDegrees=45, swing2LimitDegrees=45, twistLimitDegrees=5}
 {bone1Name=dyn_hairfrontA_01_r, bone2Name=dyn_hairfrontA_01_r_MSTR, swing1LimitDegrees=45, swing2LimitDegrees=45, twistLimitDegrees=5}
 {bone1Name=dyn_hairfrontA_02_r, bone2Name=dyn_hairfrontA_01_r, swing1LimitDegrees=45, swing2LimitDegrees=45, twistLimitDegrees=5}
 {bone1Name=dyn_hairfrontA_03_r, bone2Name=dyn_hairfrontA_02_r, swing1LimitDegrees=45, swing2LimitDegrees=45, twistLimitDegrees=5}
 {bone1Name=dyn_hairfrontB_01_r, bone2Name=dyn_hairfrontB_01_r_MSTR, swing1LimitDegrees=45, swing2LimitDegrees=45, twistLimitDegrees=5}
 {bone1Name=dyn_hairfrontB_02_r, bone2Name=dyn_hairfrontB_01_r, swing1LimitDegrees=45, swing2LimitDegrees=45, twistLimitDegrees=5}
 {bone1Name=dyn_hairfrontB_03_r, bone2Name=dyn_hairfrontB_02_r, swing1LimitDegrees=45, swing2LimitDegrees=45, twistLimitDegrees=5}
 {bone1Name=dyn_hairfrontC_01_r, bone2Name=dyn_hairfrontB_01_r, swing1LimitDegrees=45, swing2LimitDegrees=45, twistLimitDegrees=5}
 {bone1Name=dyn_hairfrontC_02_r, bone2Name=dyn_hairfrontC_01_r, swing1LimitDegrees=45, swing2LimitDegrees=45, twistLimitDegrees=5}
 {bone1Name=dyn_hairfrontD_01_r, bone2Name=dyn_hairfrontD_01_r_MSTR, swing1LimitDegrees=45, swing2LimitDegrees=45, twistLimitDegrees=5}
 {bone1Name=dyn_hairfrontD_02_r, bone2Name=dyn_hairfrontD_01_r, swing1LimitDegrees=45, swing2LimitDegrees=45, twistLimitDegrees=5}
 {bone1Name=dyn_hairfrontE_01_r, bone2Name=dyn_hairfrontE_01_r_MSTR, swing1LimitDegrees=45, swing2LimitDegrees=45, twistLimitDegrees=5}
 {bone1Name=dyn_hairfrontE_02_r, bone2Name=dyn_hairfrontE_01_r, swing1LimitDegrees=45, swing2LimitDegrees=45, twistLimitDegrees=5}
 {bone1Name=dyn_hairtop_01_r, bone2Name=dyn_hairtop_01_r_MSTR, swing1LimitDegrees=45, swing2LimitDegrees=45, twistLimitDegrees=5}
 {bone1Name=dyn_hairtop_02_r, bone2Name=dyn_hairtop_01_r, swing1LimitDegrees=45, swing2LimitDegrees=45, twistLimitDegrees=5}
 {bone1Name=dyn_hairtop_01_l, bone2Name=dyn_hairtop_01_l_MSTR, swing1LimitDegrees=45, swing2LimitDegrees=45, twistLimitDegrees=5}
 {bone1Name=dyn_hairsideB_01_l, bone2Name=dyn_hairsideB_01_l_MSTR, swing1LimitDegrees=45, swing2LimitDegrees=45, twistLimitDegrees=5}
 {bone1Name=dyn_hairsideB_02_l, bone2Name=dyn_hairsideB_01_l, swing1LimitDegrees=45, swing2LimitDegrees=45, twistLimitDegrees=5}
 {bone1Name=dyn_hairside_01_l, bone2Name=dyn_hairside_01_l_MSTR, swing1LimitDegrees=45, swing2LimitDegrees=45, twistLimitDegrees=5}
 {bone1Name=dyn_hairside_02_l, bone2Name=dyn_hairside_01_l, swing1LimitDegrees=45, swing2LimitDegrees=45, twistLimitDegrees=5}
 {bone1Name=dyn_hairside_03_l, bone2Name=dyn_hairside_02_l, swing1LimitDegrees=45, swing2LimitDegrees=45, twistLimitDegrees=5}
 {bone1Name=dyn_hairside_04_l, bone2Name=dyn_hairside_03_l, swing1LimitDegrees=45, swing2LimitDegrees=45, twistLimitDegrees=5}
 {bone1Name=zipper_01_l, bone2Name=zipper_01_l_MSTR, swing1LimitDegrees=45, swing2LimitDegrees=15, twistLimitDegrees=5}
]
