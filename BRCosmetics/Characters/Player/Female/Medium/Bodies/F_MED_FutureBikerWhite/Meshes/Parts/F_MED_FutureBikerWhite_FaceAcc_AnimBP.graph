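# Post-copy animation Blueprint for the FutureBikerWhite female face accessory mesh, adding secondary motion to hair and other parts.
# It copies the parent mesh pose, then runs two chained RigidBody simulations (spine-based, then neck-based hair top).
# Skydiving uses a variant with wind force and partial ponytail look-at corrections.
# LODs above 1 or disabled rigid bodies switch to Control Rig; initialization and crouch changes reset dynamics.
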AnimBlueprint /BRCosmetics/Characters/Player/Female/Medium/Bodies/F_MED_FutureBikerWhite/Meshes/Parts/F_MED_FutureBikerWhite_FaceAcc_AnimBP
target_skeleton=&/F_MED_FutureBikerWhite_FaceAcc_Skeleton
graphs [
 AnimGraph {nodes [
   Output Pose(Result=Blend Poses by bool(BlendPose_0=Use cached pose 'ctrlRigCache'(), BlendPose_1=Use cached pose 'simCache'(), bActiveValue=Get isMobileLOD() -> (isMobileLOD:bool)) {blendType=Linear}) {layerGroup=None}
   n1 = Get windVector() -> (windVector:Vector)
   n2 = RigidBody(ComponentPose=Local To Component(LocalPose=Use cached pose 'initCache'()), ExternalForce=@n1) {
    bClampLinearTranslationLimitToRefPose=true
    bForceDisableCollisionBetweenConstraintBodies=true
    bOverrideWorldGravity=true
    overridePhysicsAsset=&/F_MED_FutureBikerWhite_FaceAcc_MESH_Physics
    simulationSpace=BaseBoneSpace
    alphaBoolBlend {blendOption=Linear}
    baseBoneRef {boneName=spine_01}
    componentAppliedLinearAccClamp=<3000 3000 3000>
    componentLinearAccScale=<0.3 0.3 0.6>
    componentLinearVelScale=<0.3 0.3 0.6>
    overrideWorldGravity {z=-980}
   }
   n3 = RigidBody(ComponentPose=@n2, ExternalForce=@n1) {
    bClampLinearTranslationLimitToRefPose=true
    bForceDisableCollisionBetweenConstraintBodies=true
    bOverrideWorldGravity=true
    overridePhysicsAsset=&/F_MED_FutureBikerWhite_FaceAcc_MESH_hairTop_Physics
    simulationSpace=BaseBoneSpace
    alphaBoolBlend {blendOption=Linear}
    baseBoneRef {boneName=neck_01}
    componentAppliedLinearAccClamp=<3000 3000 3000>
    componentLinearAccScale=<0.3 0.3 0.6>
    componentLinearVelScale=<0.3 0.3 0.6>
    overrideWorldGravity {z=-980}
   }
   n4 = Look At - Bone: dyn_ponytail_03_l(ComponentPose=@n3, Alpha=0.3) {
    alphaBoolBlend {blendOption=Linear}
    boneToModify {boneName=dyn_ponytail_03_l}
    lookAtLocation=<8.3969 42.5842 -82.456>
    lookAtTarget {boneReference {boneName=head}}
    lookAt_Axis {axis=<1 0>}
   }
   n5 = Look At - Bone: dyn_ponytail_03_r(ComponentPose=@n4, Alpha=0.3) {
    alphaBoolBlend {blendOption=Linear}
    boneToModify {boneName=dyn_ponytail_03_r}
    lookAtLocation=<-72.5349 -62.7818 -107.1315>
    lookAtTarget {boneReference {boneName=head}}
    lookAt_Axis {axis=<1 0>}
   }
   n6 = RigidBody(ComponentPose=Local To Component(LocalPose=Use cached pose 'initCache'())) {
    bClampLinearTranslationLimitToRefPose=true
    bForceDisableCollisionBetweenConstraintBodies=true
    bOverrideWorldGravity=true
    overridePhysicsAsset=&F_MED_FutureBikerWhite_FaceAcc_MESH_Physics
    simulationSpace=BaseBoneSpace
    alphaBoolBlend {blendOption=Linear}
    baseBoneRef {boneName=spine_01}
    componentAppliedLinearAccClamp=<3000 3000 3000>
    componentLinearAccScale=<0.3 0.3 0.6>
    componentLinearVelScale=<0.3 0.3 0.6>
    overrideWorldGravity {z=-980}
   }
   n7 = RigidBody(ComponentPose=@n6) {
    bClampLinearTranslationLimitToRefPose=true
    bForceDisableCollisionBetweenConstraintBodies=true
    bOverrideWorldGravity=true
    overridePhysicsAsset=&F_MED_FutureBikerWhite_FaceAcc_MESH_hairTop_Physics
    simulationSpace=BaseBoneSpace
    alphaBoolBlend {blendOption=Linear}
    baseBoneRef {boneName=neck_01}
    componentAppliedLinearAccClamp=<3000 3000 3000>
    componentLinearAccScale=<0.3 0.3 0.6>
    componentLinearVelScale=<0.3 0.3 0.6>
    overrideWorldGravity {z=-980}
   }
   Save cached pose 'simCache'(Pose=Blend Poses by bool(BlendPose_0=Component To Local(ComponentPose=@n5), BlendPose_1=Component To Local(ComponentPose=@n7), bActiveValue=Get bIsSkyDiving()) {blendType=Linear}) {cacheName=simCache}
   Save cached pose 'initCache'(Pose=Copy Pose From Mesh() {bUseAttachedParent=true}) {cacheName=initCache}
   n8 = Control Rig(Source=Use cached pose 'initCache'()) {
    alphaBoolBlend {blendOption=Linear}
    controlRigAssetReference {blueprintRigClass=&/F_MED_FutureBikerWhite_FaceAcc_CtrlRig.F_MED_FutureBikerWhite_FaceAcc_CtrlRig_C}
   }
   Save cached pose 'ctrlRigCache'(Pose=@n8) {cacheName=ctrlRigCache}
  ]}
 EventGraph {nodes [
   Event BlueprintUpdateAnimation()
   > n1 = Sequence!(execute) -> (then_0, then_1, then_2:Exec)
   n2 = Get SkydivingExternalForceInput()
   > Set windVector(windVector:Vector=vector * float(
     A=Break Fort Anim Input Skydiving External Force(FortAnimInput_SkydivingExternalForce=@n2) -> (
      bUseSkydivingVectorForce=~
      SkydivingVectorForce=~
      HeadToPelvisDirection=~
      FloatingMultiplier=~
      FloatingAdditive=~
      DivingMultiplier=~
      DivingAdditive=~
      ParachutingMultiplier=~
      ParachutingAdditive=~
     ).SkydivingVectorForce
     B=0.8
    )) -> (Output_Get:Vector)
   n3 = GetTwistAndSwingAngleOfDeltaRotationFromRefPose(self=GetOwnerSkeletalMeshComponent(), BoneName=head)
   Set headAimingAlpha(
    execute=@n1.then_1
    headAimingAlpha:double=SelectFloat(A=MapRangeClamped(Value=@n3.OutSwingAngle, InRangeA=5, InRangeB=50), B=0, bPickA=Get bIsTargeting())
   ) -> (Output_Get:double)
   n4 = Get CurrentLOD()
   n5 = Get bRigidBodyEnabled()
   n6 = Set isMobileLOD(execute=@n1.then_2, isMobileLOD:bool=OR Boolean(A=NOT Boolean(A=@n5), B="integer > integer"(A=@n4, B=1) {is=title K2Node_CallFunction})) -> (Output_Get:bool)
   > Set isMobileLODAlpha(isMobileLODAlpha:double="To Float (Boolean)"(InBool=@n6)) -> (Output_Get:double)
   n7 = OR Boolean(A=Get bCanPlayCustomAnimations() -> (bCanPlayCustomAnimations:bool), B=Get bIsPlayingEmote())
   > Set isEmotingAlpha(isEmotingAlpha:double="To Float (Boolean)"(InBool=@n7)) -> (Output_Get:double)
   > Set isEmotingAlphaInv(isEmotingAlphaInv:double="To Float (Boolean)"(InBool=NOT Boolean(A=@n7))) -> (Output_Get:double)
   Event BlueprintInitializeAnimation()
   > Set isCrouchingAlpha(isCrouchingAlpha:double=0) -> (Output_Get:double)
   > Set isNotCrouchingAlpha(isNotCrouchingAlpha:double=1) -> (Output_Get:double)
   > Set isCrouchingWeightAlpha_0(isCrouchingWeightAlpha_0:double=0.5) -> (Output_Get:double)
   > Set crouchingLegArmorAlpha(crouchingLegArmorAlpha:double=0.5) -> (Output_Get:double)
   > Set crouchingLegArmorAlpha(crouchingLegArmorAlpha=1)
   > ResetDynamics(InTeleportType=TeleportPhysics)
   > ResetDynamics(InTeleportType=ResetPhysics)
   Event OnBeginCrouching()
   > Set isCrouchingAlpha(isCrouchingAlpha=1)
   > Set isNotCrouchingAlpha(isNotCrouchingAlpha=0)
   > Set isCrouchingWeightAlpha_0(isCrouchingWeightAlpha_0=1)
   > Set crouchingLegArmorAlpha(crouchingLegArmorAlpha=0)
   > Set crouchingLegArmorAlpha(crouchingLegArmorAlpha=0.25)
   > ResetDynamics(InTeleportType=TeleportPhysics)
   > ResetDynamics(InTeleportType=ResetPhysics)
   Event OnEndCrouching()
   > Set isCrouchingAlpha(isCrouchingAlpha=0)
   > Set isNotCrouchingAlpha(isNotCrouchingAlpha=1)
   > Set isCrouchingWeightAlpha_0(isCrouchingWeightAlpha_0=0.5)
   > Set crouchingLegArmorAlpha(crouchingLegArmorAlpha=0.5)
   > Set crouchingLegArmorAlpha(crouchingLegArmorAlpha=1)
   > Delay(Duration=0.125)
   > ResetDynamics(InTeleportType=TeleportPhysics)
   > ResetDynamics(InTeleportType=ResetPhysics)
   TryGetPawnOwner()
  ]}
]
variables [
 crouchingLegArmorAlpha {type=number}
 headAimingAlpha {type=number}
 isCrouchingAlpha {type=number}
 isCrouchingWeightAlpha_0 {type=number}
 isEmotingAlpha {type=number}
 isEmotingAlphaInv {type=number}
 isMobileLOD {type=boolean}
 isMobileLODAlpha {type=number}
 isNotCrouchingAlpha {type=number}
 windVector {type=Vector}
]
